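# Editor utility widget for Landmass that lists a brush manager's brushes and lets the user draw polygon shapes for them over a preview landscape.
# Mouse input is projected onto a Z plane and stored as normalized points, which are rasterized as a triangle fan into a render target.
# On completion, a square crop is jump-flooded into a distance field, converted to a texture and assigned to the selected brush, updating its location, size and spline.
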EditorUtilityWidgetBlueprint /Landmass/Landscape/BlueprintBrushes/Widgets/EUW_Landmass_Editor
graphs [
 Cast Cursor to Z Plane {nodes [
   n1 = GetViewLocation!(self:Viewport Object Reference=Get LandmassViewport!(self:Landmass Viewport Widget Object Reference=Get LandmassViewport_Widget() -> (LandmassViewport_Widget:Landmass Viewport Widget Object Reference)) -> (LandmassViewport:Landmass Viewport Object Reference)) -> (ReturnValue:Vector)
   n2 = BreakVector(InVec=@n1)
   n3 = Inputs!() -> (World Ray:Vector=~, Z Plane:double=~)
   float + float(A=@n3.Z Plane, B=@n2.Z)
   n4 = float - float(A=@n2.Z, B=@n3.Z Plane)
   n5 = vector / vector(A:Vector=@n3, B="Absolute (Float)"(A=BreakVector(InVec=@n3).Z)) -> (ReturnValue:Vector)
   Outputs!(Intersection:Vector=vector + vector(A=vector * vector(A=@n5, B=@n4) -> (ReturnValue=~), B=@n1))
  ]}
 Check and Create RT {nodes [
   n1 = Inputs!() -> (In:Exec, RT:Texture Render Target 2D Object Reference, Width:int, Height:int, Format:ETextureRenderTargetFormat Enum=RTF_R8)
   > n2 = Is Valid(InputObject=@n1)
   n3 = Get SizeY!(self:Texture Render Target 2D Object Reference=@n1) -> (SizeY:int=0)
   n4 = Get SizeX!(self:Texture Render Target 2D Object Reference=@n1) -> (SizeX:int=0)
   n5 = Get RenderTargetFormat!(self:Texture Render Target 2D Object Reference=@n1) -> (RenderTargetFormat:ETextureRenderTargetFormat Enum=RTF_R8)
   > n6 = Branch(Condition=AND Boolean!(
     A="Equal (Integer)"(A=@n4, B=@n1.Width) {is=title K2Node_CallFunction}
     B="Equal (Integer)"(A=@n3, B=@n1.Height)
     C:bool="Equal (Enum)"(A:ETextureRenderTargetFormat Enum=@n5, B:ETextureRenderTargetFormat Enum=@n1.Format)
    ) -> (ReturnValue))
   Outputs!(Out:Exec=[@n8.then, @n6.then])
   n7 = CreateRenderTarget2D!(
    execute=[@n6.else, @n2.Is Not Valid]
    WorldContextObject:Object Reference=Get World() -> (World:Object Reference)
    Width:int=@n1.Width
    Height:int=@n1.Height
    Format:ETextureRenderTargetFormat Enum=@n1.Format
    ClearColor:Linear Color Structure=(R=0,G=0,B=0,A=1)
    bAutoGenerateMipMaps:bool=false
    bSupportUAVs:bool=false
   ) -> (then, ReturnValue:Texture Render Target 2D Object Reference)
   > n8 = "Set Texture Render Target 2D Object Reference (by ref)"!(execute:Exec, Target:"Texture Render Target 2D Object Reference (by ref)"=@n1, Value:Texture Render Target 2D Object Reference=@n7) -> (then:Exec) {is=title K2Node_VariableSetRef}
  ]}
 Draw Debug Lines for Point Array {nodes [
   Inputs!() -> (execute:Exec, MouseVector:Vector=~)
   n1 = Get Points() -> (Points:Array of Vectors)
   > n2 = For Loop(LastIndex=int - int(A=Last Index!(TargetArray:Array of Vectors=@n1) -> (ReturnValue=0)))
   Outputs!(Completed:Exec=@n2.Completed)
   n3 = Normalized Landscape To World Coord!(A:Vector="Get (a ref)"(Array:Array of Vectors=Get Points(), Dimension 1=@n2) -> (Output:"Vector (by ref)")) -> (ReturnValue:Vector=~) {is=title K2Node_MacroInstance}
   n4 = GetViewLocation(self=Get LandmassViewport(self=Get LandmassViewport_Widget()))
   n5 = vector - vector(A=@n4, B=@n3)
   n6 = Normalized Landscape To World Coord(A="Get (a ref)"(Array=Get Points(), Dimension 1=int + int(A=@n2))) -> (ReturnValue=~)
   n7 = vector - vector(A=@n4, B=@n6)
   > DrawDebugLine!(
    execute
    WorldContextObject:Object Reference=Get World()
    LineStart:Vector=vector + vector(A=@n3, B=vector * vector(A=@n5, B:float=0.25) -> (ReturnValue=~))
    LineEnd:Vector=vector + vector(A=@n6, B=vector * vector(A=@n7, B=0.25) -> (ReturnValue=~))
    LineColor:Linear Color Structure=(R=0.11,G=1,B=0,A=1)
    Duration:float=60
    Thickness:float=512
    DepthPriority:EDrawDebugSceneDepthPriorityGroup Enum=World
   ) -> (then)
   Get Cursor World Ray!(Cursor:"Pointer Event Structure (by ref)"=Get MouseEvent() -> (MouseEvent:"Pointer Event Structure (by ref)")) -> (World Ray:Vector=~) {is=title K2Node_MacroInstance}
  ]}
 Draw Triangle {nodes [
   Draw Triangle() {is=title K2Node_FunctionEntry}
   > n1 = BeginDrawCanvasToRenderTarget!(
    execute
    WorldContextObject:Object Reference=Get World()
    TextureRenderTarget:Texture Render Target 2D Object Reference=Get ShapeRenderTarget() -> (ShapeRenderTarget:Texture Render Target 2D Object Reference)
   ) -> (then, Canvas:Canvas Object Reference, Size:Vector 2D Structure, Context:Draw To Render Target Context Structure)
   n2 = Get Points()
   n3 = Get SizeX(self=Get ShapeRenderTarget()) -> (SizeX=0)
   n4 = "To Vector2D (Vector)"!(InVector:Vector=vector * vector(A="Get (a ref)"(Array=@n2, Dimension 1=0), B:int=@n3) -> (ReturnValue=~)) -> (ReturnValue:Vector 2D Structure)
   n5 = Last Index(TargetArray=Get Points()) -> (ReturnValue=0)
   n6 = int - int(A=@n5)
   n7 = Get Points()
   n8 = "To Vector2D (Vector)"(InVector=vector * vector(A="Get (a ref)"(Array=@n7, Dimension 1=@n6), B=@n3) -> (ReturnValue=~))
   n9 = Get Points()
   n10 = "To Vector2D (Vector)"(InVector=vector * vector(A="Get (a ref)"(Array=@n9, Dimension 1=@n5), B=@n3) -> (ReturnValue=~))
   > Draw Material Triangles!(
    execute
    self:Canvas Object Reference=@n1
    RenderMaterial:Material Interface Object Reference=/Landmass/Landscape/OldPrototype_BP/Mats/Debug/Intersections
    Triangles:Array of Canvas UVTri Structures=Make Array("[0]":Canvas UVTri Structure=Make Canvas UVTri!(
      V0_Pos:Vector 2D Structure=@n4
      V0_UV:Vector 2D Structure=(X=0,Y=0)
      V0_Color:Linear Color Structure=(R=1,G=0,B=0,A=1)
      V1_Pos:Vector 2D Structure=@n10
      V1_UV:Vector 2D Structure=(X=0,Y=0)
      V1_Color:Linear Color Structure=(R=0,G=1,B=0.245,A=1)
      V2_Pos:Vector 2D Structure=@n8
      V2_UV:Vector 2D Structure=(X=0,Y=0)
      V2_Color:Linear Color Structure=(R=0.0003,G=0,B=1,A=1)
     ) -> (CanvasUVTri:Canvas UVTri Structure) {is=title K2Node_MakeStruct}) -> (Array:Array of Canvas UVTri Structures)
   ) -> (then)
   > EndDrawCanvasToRenderTarget!(execute, WorldContextObject:Object Reference=Get World(), Context:"Draw To Render Target Context Structure (by ref)"=@n1.Context) -> (then)
  ]}
 DrawingWidget_MouseDoubleClick {nodes [
   DrawingWidget_MouseDoubleClick!() -> (then, MyGeometry:Geometry Structure, MouseEvent:"Pointer Event Structure (by ref)") {is=title K2Node_FunctionEntry}
   > n1 = Sequence()
   > Branch(Condition=Get ClickToAddMode() -> (ClickToAddMode:bool))
   > Set ClickToAddMode(ClickToAddMode:bool=false) -> (Output_Get:bool)
   n2 = Last Index(TargetArray=Get Points()) -> (ReturnValue=0)
   n3 = Get Points()
   n4 = Get Points()
   > DrawDebugLine(
    WorldContextObject=Get World()
    LineStart=Normalized Landscape To World Coord(A="Get (a ref)"(Array=@n3, Dimension 1=@n2)) -> (ReturnValue=~)
    LineEnd=Normalized Landscape To World Coord(A="Get (a ref)"(Array=@n4, Dimension 1=0)) -> (ReturnValue=~)
    LineColor=(R=0.11,G=1,B=0,A=1)
    Duration=60
    Thickness=512
    DepthPriority=World
   )
   Return Node!(execute=@n1.then_1, ReturnValue:Event Reply Structure=Make Event Reply!() -> (EventReply:Event Reply Structure) {is=title K2Node_MakeStruct})
  ]}
 DrawingWidget_MouseDown {nodes [
   DrawingWidget_MouseDown!() -> (then, MyGeometry:Geometry Structure, MouseEvent:"Pointer Event Structure (by ref)") {is=title K2Node_FunctionEntry}
   > n1 = Sequence()
   > n2 = Branch(Condition=Get ClickAndDragMode() -> (ClickAndDragMode:bool))
   > Set Drawing_MouseDragging(Drawing_MouseDragging:bool=true) -> (Output_Get:bool)
   > PrintString(InString=CLICK)
   > Clear(TargetArray:Array of Vectors=Get Points())
   > FlushPersistentDebugLines!(execute, WorldContextObject:Object Reference=Get World()) -> (then)
   n3 = Branch(execute=@n2.else, Condition=Get ClickToAddMode())
   n4 = Get MouseEvent()
   n5 = World To Normalized Landscape Coord!(A:Vector=Cast Cursor To Z Plane!(World Ray:Vector=Get Cursor World Ray(Cursor=@n4) -> (World Ray=~), Z Plane:double=Get Selected Brush Z() -> (Selected Brush Z:double)) -> (Intersection:Vector=~) {is=title K2Node_MacroInstance}) -> (ReturnValue:Vector=~) {is=title K2Node_MacroInstance}
   Add(execute=[@n7.then, @n3.then], TargetArray:Array of Vectors=Get Points(), NewItem:"Vector (by ref)"=@n5) {is=title K2Node_CallArrayFunction}
   > Set ExtentsMin(ExtentsMin:Vector 2D Structure=MinV2!(
     A:Vector 2D Structure=Get ExtentsMin() -> (ExtentsMin:Vector 2D Structure)
     B:Vector 2D Structure="To Vector2D (Vector)"(InVector=@n5)
     __WorldContext:Object Reference
    ) -> (Min:Vector 2D Structure)) -> (Output_Get:Vector 2D Structure)
   > Set ExtentsMax(ExtentsMax:Vector 2D Structure=MaxV2!(
     A:Vector 2D Structure=Get ExtentsMax() -> (ExtentsMax:Vector 2D Structure)
     B:Vector 2D Structure="To Vector2D (Vector)"(InVector=@n5)
     __WorldContext:Object Reference
    ) -> (Min:Vector 2D Structure)) -> (Output_Get:Vector 2D Structure)
   n6 = Get Points()
   > Branch(Condition="integer >= integer"(A=Length(TargetArray:Array of Vectors=@n6), B=3))
   > Draw Triangle() {is=title K2Node_CallFunction}
   Set ClickToAddMode(execute=@n3.else, ClickToAddMode=true)
   > n7 = Clear(TargetArray=Get Points())
   Return Node(execute=@n1.then_1, ReturnValue=Make Event Reply())
  ]}
 DrawingWidget_MouseMove {nodes [
   n1 = Get Actor Location(self=Get SelectedBrush() -> (SelectedBrush:Landmass Actor Object Reference))
   n2 = Get MouseEvent()
   n3 = BreakVector(InVec=Cast Cursor To Z Plane(World Ray=Get Cursor World Ray(Cursor=@n2) -> (World Ray=~)) -> (Intersection=~))
   n4 = Get LandmassViewport_Widget()
   n5 = AddSplinePoint!(
    execute
    self:Spline Component Object Reference=Get PreviewSpline!(self:Landmass Widget Components Object Reference=Get WidgetComponentBP!(self:Landmass Viewport Widget Object Reference=@n4) -> (WidgetComponentBP:Landmass Widget Components Object Reference)) -> (PreviewSpline:Spline Component Object Reference)
    Position:"Vector (by ref)"=MakeVector(X=@n3, Y=@n3.Y, Z=BreakVector(InVec=@n1).Z)
    CoordinateSpace:ESplineCoordinateSpace Enum=World
    bUpdateSpline:bool=true
   ) -> (then)
   n6 = Get MouseEvent()
   n7 = World To Normalized Landscape Coord(A=Cast Cursor To Z Plane(World Ray=Get Cursor World Ray(Cursor=@n6) -> (World Ray=~), Z Plane=Get Selected Brush Z()) -> (Intersection=~)) -> (ReturnValue=~)
   Add(execute=[@n5.then, @n17.then], TargetArray=Get Points(), NewItem=@n7)
   > Set ExtentsMin(ExtentsMin=MinV2(A=Get ExtentsMin(), B="To Vector2D (Vector)"(InVector=@n7)))
   > Set ExtentsMax(ExtentsMax=MaxV2(A=Get ExtentsMax(), B="To Vector2D (Vector)"(InVector=@n7)))
   n8 = Get Points()
   > n9 = Branch(Condition="integer >= integer"(A=Length(TargetArray=@n8), B=3))
   > n10 = Draw Triangle()
   n11 = Get Points()
   Branch(execute=[@n10.then, @n9.else], Condition="integer > integer"(A=Length(TargetArray=@n11), B=1))
   DrawingWidget_MouseMove!() -> (then, MyGeometry:Geometry Structure, MouseEvent:"Pointer Event Structure (by ref)") {is=title K2Node_FunctionEntry}
   > n12 = Sequence!(execute) -> (then_0, then_1, then_2:Exec)
   > n13 = Branch(Condition=Get Drawing_MouseDragging() -> (Drawing_MouseDragging:bool))
   > n14 = Sequence!(execute) -> (then_0, then_1)
   n15 = Get Points()
   n16 = vector - vector(A="Get (a ref)"(Array=@n15, Dimension 1=Last Index(TargetArray=@n15) -> (ReturnValue=0)), B=@n7)
   > n17 = Branch(Condition="float > float"(A=Vector Length(A=@n16), B=0.02))
   n18 = Branch(execute=@n14.then_1, Condition=Get ClickAndDragMode())
   FlushPersistentDebugLines(execute=[@n27.then, @n18.then], WorldContextObject=Get World())
   n19 = Get Points()
   > Branch(Condition="integer > integer"(A=Length(TargetArray=@n19)))
   > Draw Debug Lines For Point Array!(execute:Exec, MouseVector:Vector) -> (Completed:Exec) {is=title K2Node_MacroInstance}
   n20 = Cast Cursor To Z Plane(World Ray=Get Cursor World Ray(Cursor=Get MouseEvent()) -> (World Ray=~), Z Plane=Get Selected Brush Z()) -> (Intersection=~)
   n21 = GetViewLocation(self=Get LandmassViewport(self=Get LandmassViewport_Widget()))
   n22 = vector - vector(A=@n21, B=@n20)
   n23 = Get Points()
   n24 = Normalized Landscape To World Coord(A="Get (a ref)"(Array=Get Points(), Dimension 1=Last Index(TargetArray=@n23) -> (ReturnValue=0))) -> (ReturnValue=~)
   n25 = vector - vector(A=@n21, B=@n24)
   > DrawDebugLine(
    WorldContextObject=Get World()
    LineStart=vector + vector(A=@n24, B=vector * vector(A=@n25, B:float=0.25) -> (ReturnValue=~))
    LineEnd=vector + vector(A=@n20, B=vector * vector(A=@n22, B=0.25) -> (ReturnValue=~))
    LineColor=(R=0.11,G=1,B=0,A=1)
    Duration=60
    Thickness=512
    DepthPriority=World
   )
   n26 = Branch(execute=@n13.else, Condition=Get ClickAndDragMode())
   n27 = Branch(execute=@n26.else, Condition=Get ClickToAddMode())
   Return Node(execute=@n12.then_2, ReturnValue=Make Event Reply())
  ]}
 DrawingWidget_MouseUp {nodes [
   DrawingWidget_MouseUp!() -> (then, MyGeometry:Geometry Structure, MouseEvent:"Pointer Event Structure (by ref)") {is=title K2Node_FunctionEntry}
   > n1 = Sequence()
   > Branch(Condition=Get ClickAndDragMode())
   > Set Drawing_MouseDragging(Drawing_MouseDragging=false)
   > SetTextureParameterValue(self=Get CopyRTMID() -> (CopyRTMID:Material Instance Dynamic Object Reference), ParameterName=RT, Value=Get ShapeRenderTarget())
   > DrawMaterialToRenderTarget!(
    execute
    WorldContextObject:Object Reference=Get World()
    TextureRenderTarget:Texture Render Target 2D Object Reference=Get CommitShapeRT() -> (CommitShapeRT:Texture Render Target 2D Object Reference)
    Material:Material Interface Object Reference=Get CopyRTMID()
   ) -> (then)
   > SetTextureParameterValue(self=Get CopyRTMID(), ParameterName=RT, Value=Get CommitShapeRT())
   > DrawMaterialToRenderTarget(WorldContextObject=Get World(), TextureRenderTarget=Get ShapeRenderTarget(), Material=Get CopyRTMID())
   > FlushPersistentDebugLines(WorldContextObject=Get World())
   > Draw Debug Lines For Point Array()
   n2 = Normalized Landscape To World Coord(A="Get (a ref)"(Array=Get Points(), Dimension 1=0)) -> (ReturnValue=~)
   n3 = GetViewLocation(self=Get LandmassViewport(self=Get LandmassViewport_Widget()))
   n4 = vector - vector(A=@n3, B=@n2)
   n5 = Get Points()
   n6 = Normalized Landscape To World Coord(A="Get (a ref)"(Array=Get Points(), Dimension 1=Last Index(TargetArray=@n5) -> (ReturnValue=0))) -> (ReturnValue=~)
   n7 = vector - vector(A=@n3, B=@n6)
   > DrawDebugLine(
    WorldContextObject=Get World()
    LineStart=vector + vector(A=@n6, B=vector * vector(A=@n7, B=0.25) -> (ReturnValue=~))
    LineEnd=vector + vector(A=@n2, B=vector * vector(A=@n4, B=0.25) -> (ReturnValue=~))
    LineColor=(R=0.11,G=1,B=0,A=1)
    Duration=60
    Thickness=512
    DepthPriority=World
   )
   Return Node(execute=@n1.then_1, ReturnValue=Make Event Reply())
  ]}
 EventGraph {nodes [
   "On Clicked (EditorUtilityButton_BeginDrawing)"() {is=title K2Node_ComponentBoundEvent}
   > K2Node_Composite_4 = BeginDrawingShape!(execute) -> (Drawing Started:Exec, Drawing Completed:Exec) {is=title K2Node_Composite}
   n1 = Sequence(execute=@K2Node_Composite_4.Drawing Completed)
   K2Node_Composite_2 = Make Distance Field!(In:Exec=[@n2.then, @n1.then_0]) {is=title K2Node_Composite}
   K2Node_Composite_3 = SaveShapeToBrush(execute=[@n6.then, @n1.then_1]) {is=title K2Node_Composite}
   n2 = "On Clicked (EditorUtilityButton_MakeDF)"() {is=title K2Node_ComponentBoundEvent}
   n3 = "On Item Clicked (EditorUtilityListView_187)"!() -> (OutputDelegate, then, Item:Object Reference) {is=title K2Node_ComponentBoundEvent}
   > n4 = Branch(Condition="Equal (Object)"(A=@n3.Item, B=Get SelectedObject() -> (SelectedObject:Object Reference)))
   > SetObject(self=Get Details_ItemFromList() -> (Details_ItemFromList:Details View Object Reference))
   > Set SelectedObject(SelectedObject:Object Reference) -> (Output_Get:Object Reference)
   > SetVisibility(self=Get VerticalBox_Drawshape() -> (VerticalBox_Drawshape:Vertical Box Object Reference), InVisibility=Collapsed)
   n5 = Set SelectedObject(execute=@n4.else, SelectedObject=@n3.Item)
   SetObject(execute=[@n7.then, @n5.then], self=Get Details_ItemFromList(), NewObject=@n5)
   > SetPreviewPlaneToBrush()
   > SetVisibility(self=Get VerticalBox_Drawshape(), InVisibility=SelfHitTestInvisible)
   n6 = "On Clicked (Button_SaveShapeToBrush)"() {is=title K2Node_ComponentBoundEvent}
   n7 = Select Brush() {is=title K2Node_CustomEvent}
   Add Brushes to List() {is=title K2Node_CustomEvent}
   > n8 = For Each Loop(Array:Array of Actor Object References=Get Brushes() -> (Brushes:Array of Actor Object References)) -> (Array Element:Actor Object Reference)
   > AddItem!(
    execute
    self:List View Object Reference=Get EditorUtilityListView_187() -> (EditorUtilityListView_187:Editor Utility List View Object Reference)
    Item:Object Reference=@n8
   ) -> (then)
   Event Construct()
   > Delay!(execute, WorldContextObject:Object Reference, Duration=0.005) -> (then)
   > n9 = Sequence!(execute) -> (then_0, then_1, then_2:Exec, then_3:Exec)
   n10 = Get LandmassViewport_Widget()
   > Set World(World:Object Reference=GetViewportWorld!(self:Viewport Object Reference=Get LandmassViewport(self=@n10)) -> (ReturnValue:World Object Reference)) -> (Output_Get:Object Reference)
   > SetObject(
    self=[
     Get SinglePropertyView_CroppedRT() -> (SinglePropertyView_CroppedRT:Single Property View Object Reference)
     Get SinglePropertyView_ShapeRT() -> (SinglePropertyView_ShapeRT:Single Property View Object Reference)
     Get SinglePropertyView_DFRT() -> (SinglePropertyView_DFRT:Single Property View Object Reference)
    ]
    NewObject=Self-Reference()
   )
   n11 = Get BrushManager() -> (BrushManager:Landmass Brush Manager Object Reference)
   > Is Valid(InputObject=@n11)
   > Set Brushes(Brushes:Array of Actor Object References=Get LandmassBrushes!(self:Landmass Manager Base Object Reference=@n11) -> (LandmassBrushes:Array of Landmass Actor Object References)) -> (Output_Get:Array of Actor Object References)
   > Add Brushes to List() {is=title K2Node_CallFunction}
   > K2Node_Composite_1 = Ensure Preview Landscape Data Exists() {is=title K2Node_Composite}
   Is Valid(exec=@n9.then_2, InputObject=Get BrushManager())
   > K2Node_Composite_5 = Focus Camera on PreviewLandscape!(execute) -> (then:Exec) {is=title K2Node_Composite}
   > K2Node_Composite_0 = Spawn Drawing and Brush Selection Meshes!(execute) -> (then:Exec) {is=title K2Node_Composite}
   n12 = Get SelectedBrush()
   > Is Valid(InputObject=@n12)
   > n13 = Set SelectedObject(SelectedObject=@n12)
   > Set Selected Item!(execute, self:List View Object Reference=Get EditorUtilityListView_187(), Item:Object Reference=@n13) -> (then) {is=title K2Node_CallFunction}
   > Select Brush() {is=title K2Node_CallFunction}
   Is Valid(exec=@n9.then_1, InputObject=Get BrushManager())
   n14 = Get LandmassViewport_Widget()
   > n15 = Set PreviewLandscape(PreviewLandscape:Landmass Preview Landscape Object Reference=Get PreviewLandscape!(self:Landmass Widget Components Object Reference=Get WidgetComponentBP(self=@n14)) -> (PreviewLandscape:Landmass Preview Landscape Object Reference)) -> (Output_Get:Landmass Preview Landscape Object Reference)
   > Spawn Instance Meshes!(execute, self:Landmass Preview Landscape Object Reference=@n15, Brush Manager:Landmass Brush Manager Object Reference=Get BrushManager()) -> (then)
   > SetVisibility!(
    execute
    self:Scene Component Object Reference=Get Landscape ISMC!(self:Landmass Preview Landscape Object Reference=Get PreviewLandscape!() -> (PreviewLandscape)) -> (Landscape ISMC:Hierarchical Instanced Static Mesh Component Object Reference)
    bNewVisibility:bool=true
    bPropagateToChildren:bool=false
   ) -> (then)
   > Update MID Height!(
    execute
    self:Landmass Preview Landscape Object Reference=Get PreviewLandscape()
    Height RT:Texture Object Reference=Get Heightmap Render RT!(self:Landmass Brush Manager Object Reference=Get BrushManager()) -> (Heightmap Render RT:Texture Render Target 2D Object Reference)
   ) -> (then)
   > Set MPC params(self:Landmass Preview Landscape Object Reference=Get PreviewLandscape()) {is=title K2Node_CallFunction}
   n16 = Event MoveDown!() -> (OutputDelegate, then, Brush:Actor Object Reference) {is=title K2Node_Event}
   n17 = Get Brushes()
   n18 = Find Item!(TargetArray:Array of Actor Object References=@n17, ItemToFind:"Actor Object Reference (by ref)"=@n16.Brush) -> (ReturnValue)
   > Branch(Condition="integer < integer"(A=@n18, B=Last Index(TargetArray:Array of Actor Object References=@n17) -> (ReturnValue=0)))
   > Swap Array Elements!(execute, TargetArray:Array of Actor Object References=Get Brushes(), FirstIndex:int=@n18, SecondIndex:int=int + int(A=@n18)) -> (then) {is=title K2Node_CallArrayFunction}
   > ClearListItems(self=Get EditorUtilityListView_187())
   > Add Brushes to List()
   n19 = Event MoveUp!() -> (OutputDelegate, then, Brush:Actor Object Reference) {is=title K2Node_Event}
   n20 = Find Item(TargetArray=Get Brushes(), ItemToFind=@n19.Brush)
   > Branch(Condition="integer > integer"(A=@n20))
   > Swap Array Elements(TargetArray=Get Brushes(), FirstIndex=@n20, SecondIndex=int - int(A=@n20))
   > ClearListItems(self=Get EditorUtilityListView_187())
   > Add Brushes to List()
   SpawnActor Custom Brush Material Only!(
    execute:Exec
    WorldContextObject:Object Reference=Self-Reference()
    Class:Actor Class Reference=/Landmass/Landscape/BlueprintBrushes/CustomBrush_MaterialOnly.CustomBrush_MaterialOnly_C
    SpawnTransform:Transform=MakeTransform(Rotation="0, 0, 0")
    CollisionHandlingOverride:ESpawnActorCollisionHandlingMethod Enum=Undefined
    TransformScaleMethod:ESpawnActorScaleMethod Enum=MultiplyWithRoot
    Owner:Actor Object Reference
    Instigator:Pawn Object Reference
   ) -> (then:Exec, ReturnValue:Custom Brush Material Only Object Reference) {is=title K2Node_SpawnActorFromClass}
  ]}
 Spawn Drawing and Brush Selection Meshes {
  path=EventGraph.K2Node_Composite_0.Spawn Drawing and Brush Selection Meshes
  nodes [
   Inputs!() -> (execute)
   > n1 = Sequence!(execute) -> (then_0, then_1)
   n2 = Get LandscapeInformation!(self:Landmass Manager Base Object Reference=Get BrushManager()) -> (LandscapeInformation:Landmass Landscape Info Structure)
   n3 = Get World Size!(self:Landmass Brush Manager Object Reference=Get BrushManager()) -> (World Size:Vector)
   n4 = vector + vector(
    A=BreakTransform(InTransform=Break Landmass Landscape Info!(LandmassLandscapeInfo:"Landmass Landscape Info Structure (by ref)"=@n2) -> (LandscapeTransform:Transform, LandscapeQuads:Int Point Structure, RenderTargetResolution:Int Point Structure) {is=title K2Node_BreakStruct})
    B=vector / vector(A=@n3, B=2)
   )
   > n5 = SpawnActor Static Mesh Actor!(
    execute:Exec
    WorldContextObject:Object Reference=Get World()
    Class:Actor Class Reference=&StaticMeshActor
    SpawnTransform:Transform=MakeTransform(Location=vector + vector(A=@n4, B="0,0,64"), Rotation="0, 0, 0", Scale=@n3)
    CollisionHandlingOverride:ESpawnActorCollisionHandlingMethod Enum=Undefined
    TransformScaleMethod:ESpawnActorScaleMethod Enum=MultiplyWithRoot
    Owner:Actor Object Reference
    Instigator:Pawn Object Reference
   ) -> (then:Exec, ReturnValue:Static Mesh Actor Object Reference) {is=title K2Node_SpawnActorFromClass}
   > n6 = Set ShapeDrawingMesh(ShapeDrawingMesh:Static Mesh Component Object Reference=Get StaticMeshComponent(self=@n5)) -> (Output_Get:Static Mesh Component Object Reference)
   > SetStaticMesh(self=@n6, NewMesh=/Engine/ArtTools/RenderToTexture/Meshes/S_1_Unit_Plane)
   > SetVisibility(self=Get ShapeDrawingMesh() -> (ShapeDrawingMesh:Static Mesh Component Object Reference), bNewVisibility=false, bPropagateToChildren=false)
   n7 = SpawnActor Static Mesh Actor(
    execute=@n1.then_1
    WorldContextObject=Get World()
    Class=&StaticMeshActor
    SpawnTransform=MakeTransform(Rotation="0, 0, 0")
    CollisionHandlingOverride=Undefined
    TransformScaleMethod=MultiplyWithRoot
   )
   > Set BrushExtentsMesh(BrushExtentsMesh:Static Mesh Component Object Reference=Get StaticMeshComponent(self=@n7)) -> (Output_Get:Static Mesh Component Object Reference)
   > SetMobility!(execute, self:Static Mesh Actor Object Reference=@n7, InMobility:EComponentMobility Enum=Movable) -> (then)
   > SetStaticMesh(
    self=Get BrushExtentsMesh() -> (BrushExtentsMesh:Static Mesh Component Object Reference)
    NewMesh=&S_1_Unit_Plane
   )
   > Create And Assign MID!(
    execute:Exec
    MID:Material Instance Dynamic Object Reference=Get BrushExtentsMID() -> (BrushExtentsMID:Material Instance Dynamic Object Reference)
    Parent:Material Interface Object Reference=/Landmass/Landscape/BlueprintBrushes/Materials/Internal/BrushBounds
    Optional Name:Name=none
    Creation Flags:EMIDCreationFlags Enum=None
   ) -> (Out:Exec) {is=title K2Node_MacroInstance}
   > SetMaterial(self=Get BrushExtentsMesh(), ElementIndex=0, Material=Get BrushExtentsMID())
   > Outputs!(then:Exec)
  ]
 }
 Ensure Preview Landscape Data Exists {
  path=EventGraph.K2Node_Composite_1.Ensure Preview Landscape Data Exists
  nodes [
   Inputs()
   > n1 = Branch(Condition=Get Enable Fast Preview Mode!(self:Landmass Brush Manager Object Reference=Get BrushManager()) -> (Enable Fast Preview Mode:bool))
   n2 = Get BrushManager()
   Set Enable Fast Preview Mode!(execute=@n1.else, Enable Fast Preview Mode:bool=true, self:Landmass Brush Manager Object Reference=@n2) -> (then, Output_Get:bool)
   > Allocate RTs(self:Landmass Brush Manager Object Reference=@n2)
   > Update Brushes Internal(self:Landmass Brush Manager Object Reference=@n2)
   > Set Enable Fast Preview Mode(Enable Fast Preview Mode=false, self=@n2)
   Outputs!()
  ]
 }
 Make Distance Field {
  path=EventGraph.K2Node_Composite_2.Make Distance Field
  nodes [
   Inputs!() -> (In:Exec)
   n1 = Break Int Point!(IntPoint:"Int Point Structure (by ref)"=Get Landscape RT Res!(self:Landmass Brush Manager Object Reference=Get BrushManager()) -> (Landscape RT Res:Int Point Structure)) -> (X:int, Y:int) {is=title K2Node_BreakStruct}
   n2 = GetMax2D!(A:Vector 2D Structure=vector2d - vector2d(A:Vector 2D Structure=Get ExtentsMax(), B:Vector 2D Structure=Get ExtentsMin()) -> (ReturnValue:Vector 2D Structure) {is=title K2Node_PromotableOperator}) -> (ReturnValue:double)
   n3 = Ceil(A=float / float(A=float * float(A:int=@n1, B=@n2) {is=title K2Node_PromotableOperator}, B=4)) {is=title K2Node_CallFunction}
   > n4 = Branch(Condition="integer > integer"(A=@n3, B=2))
   n5 = "Min (Integer)"(A=int * int(A=@n3, B=4), B=2048) -> (ReturnValue=0)
   > Check And Create RT!(
    In:Exec
    RT:Texture Render Target 2D Object Reference=Get CroppedRT() -> (CroppedRT:Texture Render Target 2D Object Reference)
    Width:int=@n5
    Height:int=@n5
    Format:ETextureRenderTargetFormat Enum=RTF_R32f
   ) -> (Out:Exec) {is=title K2Node_MacroInstance}
   > SetTextureParameterValue(self=Get CopyRTMID(), ParameterName=RT, Value=Get ShapeRenderTarget())
   > n6 = BeginDrawCanvasToRenderTarget(WorldContextObject=Get World(), TextureRenderTarget=Get CroppedRT())
   n7 = Get BrushManager()
   n8 = "To Vector2D (Vector)"(InVector=BreakTransform(InTransform=Get LS Transform!(self:Landmass Brush Manager Object Reference=@n7) -> (LS Transform:Transform)))
   n9 = Get ExtentsMin()
   n10 = Get ExtentsMax()
   n11 = vector2d / vector2d(
    A:Vector 2D Structure=vector2d + vector2d(A:Vector 2D Structure=@n9, B:Vector 2D Structure=@n10) -> (ReturnValue:Vector 2D Structure) {is=title K2Node_PromotableOperator}
    B:Vector 2D Structure=(X=2,Y=2)
   ) -> (ReturnValue:Vector 2D Structure) {is=title K2Node_PromotableOperator}
   n12 = vector2d / vector2d(A:double=GetMax2D(A=vector2d - vector2d(A=@n10, B=@n9)), B:double=2)
   n13 = Get RT World Size!(self:Landmass Brush Manager Object Reference=@n7) -> (RT World Size:Vector 2D Structure)
   n14 = vector2d * vector2d(A:Vector 2D Structure=vector2d + vector2d(A=@n11, B=@n12), B:Vector 2D Structure=@n13) -> (ReturnValue:Vector 2D Structure) {is=title K2Node_PromotableOperator}
   n15 = vector2d * vector2d(A=vector2d - vector2d(A=@n11, B=@n12), B=@n13)
   n16 = World Extents to Landmass Coordinates!(
    WorldExtents:Vector 4 Structure=MakeExtents!(
     Min:Vector 2D Structure=vector2d + vector2d(A=@n15, B=@n8)
     Max:Vector 2D Structure=vector2d + vector2d(A=@n14, B=@n8)
     __WorldContext:Object Reference
    ) -> (Extents:Vector 4 Structure)
    LandscapeInfo:Landmass Landscape Info Structure=Get LandscapeInformation(self=Get BrushManager())
   ) -> (
    ScreenPosition:Vector 2D Structure
    ScreenSize:Vector 2D Structure
    CoordinatePosition:Vector 2D Structure
    CoordinateSize:Vector 2D Structure
   )
   > Draw Material!(
    execute
    self:Canvas Object Reference=@n6
    RenderMaterial:Material Interface Object Reference=Get CopyRTMID()
    ScreenPosition:Vector 2D Structure
    ScreenSize:Vector 2D Structure=@n6.Size
    CoordinatePosition:Vector 2D Structure=@n16.CoordinatePosition
    CoordinateSize:Vector 2D Structure=@n16.CoordinateSize
    Rotation:float=0
    PivotPoint:Vector 2D Structure=(X=0.5,Y=0.5)
   ) -> (then)
   > EndDrawCanvasToRenderTarget(WorldContextObject=Get World(), Context=@n6.Context)
   n17 = Get Text(self=Get EditorUtilityEditableTextBox_BlurSteps() -> (EditorUtilityEditableTextBox_BlurSteps:Editor Utility Editable Text Box Object Reference))
   n18 = Get LandmassViewport_Widget()
   > n19 = Jump Flood and Compute Distance!(
    execute
    self:Jump Flood Component 2D Object Reference=Get JumpFlood_Component2D_Landmass!(self:Landmass Widget Components Object Reference=Get WidgetComponentBP(self=@n18)) -> (JumpFlood_Component2D_Landmass:Jump Flood Component 2D Object Reference)
    Seed:Texture Render Target 2D Object Reference=Get CroppedRT()
    Curl Params:Linear Color Structure=MakeColor!(R:float=16, G:float=0, B:float=5, A:float=0) -> (ReturnValue:Linear Color Structure)
    In Blur Passes:int=String To Integer(InString="To String (Text)"(InText=@n17))
   ) -> (then, Distance RT:Texture Render Target 2D Object Reference)
   > Set DistanceFieldRT(DistanceFieldRT:Texture Render Target 2D Object Reference=@n19) -> (Output_Get:Texture Render Target 2D Object Reference)
   PrintString(execute=@n4.else, InString=Shape Finalization Abandoned because resulting shape was too small.)
   "To Vector2D (Vector)"(InVector=Get World Size(self=@n7))
   Outputs()
   Get ShapeRenderTarget()
   Get DistanceFieldRT() -> (DistanceFieldRT:Texture Render Target 2D Object Reference)
  ]
 }
 SaveShapeToBrush {
  path=EventGraph.K2Node_Composite_3.SaveShapeToBrush
  nodes [
   n1 = PrintString(InString=UPDATE)
   Update Brushes(execute=[@n10.then, @n1.then], self:Landmass Brush Manager Object Reference=Get BrushManager())
   > SetPreviewPlaneToBrush()
   n2 = Get PreviewSpline(self=Get WidgetComponentBP(self=Get LandmassViewport_Widget()))
   n3 = GetNumberOfSplinePoints!(self:Spline Component Object Reference=@n2) -> (ReturnValue:int=0)
   n4 = For Loop(LastIndex=int - int(A=@n3))
   n5 = GetSplinePointAt!(self:Spline Component Object Reference=@n2, PointIndex:int=@n4, CoordinateSpace:ESplineCoordinateSpace Enum=World) -> (ReturnValue:Spline Point Structure)
   > PrintString(InString="To String (Vector)"!(InVec:Vector=Break Spline Point!(SplinePoint:"Spline Point Structure (by ref)"=@n5) -> (
      InputKey:float=~
      Position:Vector=~
      ArriveTangent:Vector=~
      LeaveTangent:Vector=~
      Rotation:Rotator=~
      Scale:Vector=~
      Type:ESplinePointType Enum
     ) {is=title K2Node_BreakStruct}.Position) -> (ReturnValue:String))
   Inputs!() -> (execute:Exec)
   > n6 = Sequence()
   > Branch(Condition=Does Object Implement Interface!(
     TestObject:Object Reference=Get SelectedObject()
     Interface:Interface Interface=/Landmass/Landscape/BlueprintBrushes/Interface/Landmass_Interface.Landmass_Interface_C
    ) -> (ReturnValue:bool))
   > K2Node_Composite_1 = Update Brush Location And Extents!(execute) -> (then:Exec) {is=title K2Node_Composite}
   > n7 = Construct Object from Class(Class=/Script/Engine.Texture2D, self=Get SelectedObject()) -> (ReturnValue:Texture 2D Object Reference)
   > ConvertRenderTargetToTexture2DEditorOnly(WorldContextObject=Get SelectedObject(), RenderTarget=Get DistanceFieldRT(), Texture=@n7)
   > Set CompressionSettings(CompressionSettings=TC_SingleFloat, self=@n7)
   > ForceUpdateTexture!(execute, InTexture:Texture Object Reference=@n7) -> (then)
   > Set Shape Mask!(
    execute:Exec
    self:Object Reference=Get SelectedObject()
    Texture:Texture 2D Object Reference=@n7
    Points:Array of Vectors=Get Points()
    NewExtents:Vector 4 Structure=MakeExtents()
   ) -> (then:Exec) {is=title K2Node_Message}
   > n8 = Get Shape Mask!(execute:Exec, self:Object Reference=Get SelectedBrush()) -> (then:Exec, ShapeMask:Texture 2D Object Reference) {is=title K2Node_Message}
   > SetTextureParameterValue(self=Get BrushExtentsMID(), ParameterName=ShapeRT, Value=@n8)
   > n9 = Sequence()
   > n10 = Delay(WorldContextObject=Get World(), Duration=0.1)
   Delay(execute=@n9.then_1, WorldContextObject=Get World(), Duration=0.5)
   > Update Brushes(self=Get BrushManager())
   FlushPersistentDebugLines(execute=@n6.then_1, WorldContextObject=Get World())
   Append(A="Num Spline Points: ", B="To String (Integer)"(InInt=@n3))
   Outputs()
   Get CroppedRT()
  ]
 }
 Update Brush Location And Extents {
  path=EventGraph.K2Node_Composite_3.SaveShapeToBrush.K2Node_Composite_1.Update Brush Location And Extents
  nodes [
   Inputs()
   > n1 = Sequence()
   > n2 = Cast To LandmassActor!(execute, Object=Get SelectedObject()) -> (then, CastFailed, AsLandmass Actor:Landmass Actor Object Reference) {is=title K2Node_DynamicCast}
   > n3 = Set SelectedBrush(SelectedBrush:Landmass Actor Object Reference=@n2) -> (Output_Get:Landmass Actor Object Reference)
   n4 = Get Actor Location(self=@n3)
   n5 = Get BrushManager()
   n6 = "To Vector2D (Vector)"(InVector=BreakTransform(InTransform=Get LS Transform(self=@n5)))
   n7 = "To Vector2D (Vector)"(InVector=Get World Size(self=@n5))
   n8 = vector2d + vector2d(A=vector2d * vector2d(A=Get ExtentsMin(), B=@n7), B=@n6)
   n9 = vector2d + vector2d(A=vector2d * vector2d(A=Get ExtentsMax(), B=@n7), B=@n6)
   n10 = vector2d + vector2d(A=@n8, B=@n9)
   > Set Actor Location!(
    execute
    self:Actor Object Reference=Get SelectedBrush()
    NewLocation:Vector="To Vector (Vector2D)"!(
     InVector2D:Vector 2D Structure=vector2d / vector2d(A:Vector 2D Structure=@n10, B:Vector 2D Structure=(X=2,Y=2))
     Z:float=BreakVector(InVec=@n4).Z
    ) -> (ReturnValue:Vector)
    bSweep:bool=false
    bTeleport:bool=false
   ) -> (then, SweepHitResult:Hit Result Structure, ReturnValue:bool) {is=title K2Node_CallFunction}
   n11 = Get SelectedBrush()
   > Set BrushSize!(execute, BrushSize:float=GetMax2D(A=vector2d - vector2d(A=@n9, B=@n8)), self:Landmass Actor Object Reference=@n11) -> (then, Output_Get:float)
   > Set Actor Rotation!(execute, self:Actor Object Reference=@n11, NewRotation:Rotator="0, 0, 0", bTeleportPhysics:bool=false) -> (then, ReturnValue:bool) {is=title K2Node_CallFunction}
   > SetActorScale3D!(execute, self:Actor Object Reference=Get SelectedBrush(), NewScale3D:Vector=MakeVector(X=1, Y=1, Z=1)) -> (then)
   > UpdateBrushExtents(self:Landmass Actor Object Reference=Get SelectedBrush())
   n12 = GetComponentByClass!(self:Actor Object Reference=Get SelectedBrush(), ComponentClass:Actor Component Class Reference=/Script/Engine.SplineComponent) -> (ReturnValue:Spline Component Object Reference)
   > ClearSplinePoints!(execute, self:Spline Component Object Reference=@n12, bUpdateSpline:bool=false) -> (then)
   > n13 = For Each Loop(Array:Array of Vectors=Get Points()) -> (Array Element:Vector)
   UpdateSpline(execute=@n13.Completed, self:Spline Component Object Reference=@n12)
   n14 = Get Actor Location(self=Get SelectedBrush())
   n15 = BreakVector(InVec=Normalized Landscape To World Coord(A=@n13) -> (ReturnValue=~))
   AddSplinePoint(
    execute=@n13.LoopBody
    self=@n12
    Position=MakeVector(X=@n15, Y=@n15.Y, Z=BreakVector(InVec=@n14).Z)
    CoordinateSpace=World
    bUpdateSpline=false
   )
   n16 = MakeVector(Y=0)
   > SetTangentsAtSplinePoint!(
    execute
    self:Spline Component Object Reference=@n12
    PointIndex:int=@n13.Array Index
    InArriveTangent:"Vector (by ref)"=@n16
    InLeaveTangent:"Vector (by ref)"=@n16
    CoordinateSpace:ESplineCoordinateSpace Enum=Local
    bUpdateSpline:bool=true
   ) -> (then)
   Outputs!(then:Exec=@n1.then_1)
   n17 = BreakVector2D()
   SelectVector(
    A=MakeVector(X=float / float(A=@n17, B=@n17.Y), Y=1, Z=1)
    B=MakeVector(X=1, Y=float / float(A=@n17.Y, B=@n17), Z=1)
    bPickA="float > float"(A=@n17, B=@n17.Y)
   )
  ]
 }
 BeginDrawingShape {
  path=EventGraph.K2Node_Composite_4.BeginDrawingShape
  nodes [
   Inputs()
   > n1 = Sequence()
   n2 = Break Int Point(IntPoint=Get Landscape RT Res(self=Get BrushManager()))
   > Check And Create RT(RT=Get ShapeRenderTarget(), Width=@n2, Height=@n2.Y, Format=RTF_R16f)
   n3 = Break Int Point(IntPoint=Get Landscape RT Res(self=Get BrushManager()))
   > Check And Create RT(RT=Get CommitShapeRT(), Width=@n3, Height=@n3.Y, Format=RTF_R8)
   > Create And Assign MID(
    MID=Get Shape Preview MID() -> (Shape Preview MID:Material Instance Dynamic Object Reference)
    Parent=/Landmass/Landscape/BlueprintBrushes/Materials/Internal/PreviewShapeRT
    Optional Name=none
    Creation Flags=None
   )
   > Create And Assign MID(MID=Get CopyRTMID(), Parent=/Landmass/Landscape/BlueprintBrushes/Materials/Internal/CopyRT_Shape, Optional Name=none, Creation Flags=None)
   > SetTextureParameterValue(self=Get Shape Preview MID(), ParameterName=RT, Value=Get ShapeRenderTarget())
   > SetMaterial(self=Get ShapeDrawingMesh(), ElementIndex=0, Material=Get Shape Preview MID())
   > SetVisibility(self=Get ShapeDrawingMesh(), bNewVisibility=true, bPropagateToChildren=false)
   n4 = Branch(execute=@n1.then_1, Condition=NOT Boolean(A=Get DrawingMode() -> (DrawingMode:bool)))
   > Set DrawingMode(DrawingMode:bool=true) -> (Output_Get:bool)
   > Set Text(Text="LOCTEXT(\"Complete Shape\")", self=Get DrawingButton_Text() -> (DrawingButton_Text:Text Object Reference))
   > Clear(TargetArray=Get Points())
   n5 = Get LandmassViewport_Widget()
   > ClearSplinePoints(self=Get PreviewSpline(self=Get WidgetComponentBP(self=@n5)), bUpdateSpline=true)
   > Set ExtentsMin(ExtentsMin=(X=1000000000,Y=1000000000))
   > Set ExtentsMax(ExtentsMax=(X=-1000000000,Y=-1000000000))
   > SetVisibility!(
    execute
    self:Widget Object Reference=Get MouseCapture_BorderWidget() -> (MouseCapture_BorderWidget:Border Object Reference)
    InVisibility:ESlateVisibility Enum=Visible
   ) -> (then)
   > ClearRenderTarget2D!(
    execute
    WorldContextObject:Object Reference=Get World()
    TextureRenderTarget:Texture Render Target 2D Object Reference=Get ShapeRenderTarget()
    ClearColor:Linear Color Structure=(R=0,G=0,B=0,A=1)
   ) -> (then)
   > SetScalarParameterValue(self=Get Shape Preview MID(), ParameterName=Show Distance Field, Value=0)
   n6 = "Equal (Integer)"(A=GetSelectedIndex!(self:"Combo Box (String) Object Reference"=Get EditorUtilityComboBoxString_ClickDragMode() -> (EditorUtilityComboBoxString_ClickDragMode:Editor Utility Combo Box String Object Reference)) -> (ReturnValue:int=0)) {is=title K2Node_PromotableOperator}
   > Set ClickAndDragMode(ClickAndDragMode:bool=@n6) -> (Output_Get:bool)
   > Set ClickToAddMode(ClickToAddMode=NOT Boolean(A=@n6))
   > Outputs!(Drawing Started:Exec, Drawing Completed:Exec=@n8.then)
   Set DrawingMode(execute=@n4.else, DrawingMode=false)
   > Set Text(Text="LOCTEXT(\"Begin Drawing Shape\")", self=Get DrawingButton_Text())
   > SetVisibility(self=Get MouseCapture_BorderWidget(), InVisibility=Collapsed)
   n7 = Get Points()
   > n8 = Branch(Condition="integer > integer"(A=Length(TargetArray=@n7), B=3))
  ]
 }
 Focus Camera on PreviewLandscape {
  path=EventGraph.K2Node_Composite_5.Focus Camera on PreviewLandscape
  nodes [
   Inputs()
   n1 = Get BrushManager()
   n2 = Get LS Transform(self=@n1)
   n3 = vector / vector(A=Get World Size(self=@n1), B=2)
   > Set Camera LookAt!(
    execute
    self:Landmass Viewport Widget Object Reference=Get LandmassViewport_Widget()
    FocusPoint:Vector=vector + vector(A=BreakTransform(InTransform=@n2), B=@n3)
    Distance:double=float * float(A:double=BreakVector(InVec=@n3), B=1.1)
    ViewDirection:Vector=Normalize!(A:Vector=MakeVector(Y=-0.1, Z=-1), Tolerance:float=0.0001) -> (ReturnValue:Vector) {is=title K2Node_CallFunction}
   ) -> (then) {is=title K2Node_CallFunction}
   > Outputs!(then:Exec)
  ]
 }
 Get Cursor World Ray {nodes [
   n1 = GetViewRotation!(self:Viewport Object Reference=Get LandmassViewport(self=Get LandmassViewport_Widget())) -> (ReturnValue:Rotator)
   n2 = GetCachedGeometry!(self:Widget Object Reference=Get LandmassViewport_Widget()) -> (ReturnValue:"Geometry Structure (by ref)")
   n3 = vector2d / vector2d(A=GetAbsoluteSize!(Geometry:"Geometry Structure (by ref)"=@n2) -> (ReturnValue:Vector 2D Structure), B:double=2)
   n4 = "Absolute to Local (Coordinates)"!(
    Geometry:"Geometry Structure (by ref)"=@n2
    AbsoluteCoordinate:Vector 2D Structure=Get Screen Space Position!(Input:"Pointer Event Structure (by ref)"=Inputs!() -> (Cursor:"Pointer Event Structure (by ref)")) -> (ReturnValue:Vector 2D Structure) {is=title K2Node_CallFunction}
   ) -> (ReturnValue:Vector 2D Structure)
   n5 = BreakVector2D(InVec=vector2d / vector2d(A=vector2d - vector2d(A=@n4, B=@n3), B=BreakVector2D(InVec=@n3)))
   n6 = MakeVector(X=1, Y=@n5, Z=float * float(A=@n5.Y, B=-1))
   Outputs!(World Ray:Vector=TransformDirection!(
     T:"Transform (by ref)"="To Transform (Rotator)"!(InRotator:"Rotator (by ref)"=@n1) -> (ReturnValue:Transform)
     Direction:Vector=Normalize(A=@n6, Tolerance=0.0001)
    ) -> (ReturnValue:Vector))
  ]}
 Get Shape Mask {nodes [
   Get Shape Mask() {is=title K2Node_FunctionEntry}
   > Return Node!(execute, ShapeMask:Texture 2D Object Reference)
  ]}
 GetRenderDependencies {nodes [
   GetRenderDependencies() {is=title K2Node_FunctionEntry}
   > Return Node!(execute, Objects:Array of Object References)
  ]}
 MyEUWFunction {nodes [MyEUWFunction() {is=title K2Node_FunctionEntry}]}
 Normalized Landscape to World Coord {nodes [
   n1 = Get BrushManager()
   n2 = BreakVector2D(InVec=Get RT World Size(self=@n1))
   MakeVector(X=@n2, Y=@n2.Y, Z=1)
   n3 = Break Landmass Landscape Info(LandmassLandscapeInfo=Get LandscapeInformation(self=Get BrushManager()))
   n4 = Inputs!() -> (A:Vector=~)
   n5 = Get World Size(self=@n1)
   Outputs!(ReturnValue:Vector=vector + vector(A=vector * vector(A=@n4, B:Vector=@n5) -> (ReturnValue=~), B=BreakTransform(InTransform=@n3)))
  ]}
 OnGenerateContentWidget {nodes [
   OnGenerateContentWidget!() -> (then, Item:Name) {is=title K2Node_FunctionEntry}
   > Return Node!(execute, ReturnValue:Widget Object Reference)
  ]}
 OnGenerateItemWidget {nodes [
   OnGenerateItemWidget!() -> (then, Item:Name) {is=title K2Node_FunctionEntry}
   > Return Node()
  ]}
 Render Erosion Layer {nodes [
   Render Erosion Layer!() -> (then, In Parameters:Landscape Brush Parameters Structure, Extents:Vector 4 Structure, BrushManager:Landmass Brush Manager Object Reference) {is=title K2Node_FunctionEntry}
   > Return Node!(execute, returned:bool=false)
  ]}
 SetPreviewPlaneToBrush {nodes [
   SetPreviewPlaneToBrush() {is=title K2Node_FunctionEntry}
   > n1 = Cast To LandmassActor(Object=Get SelectedObject())
   > Set SelectedBrush(SelectedBrush=@n1)
   n2 = Get SelectedBrush()
   > Set Selected Brush Z(Selected Brush Z:double=BreakVector(InVec=Get Actor Location(self=@n2)).Z) -> (Output_Get:double)
   > n3 = Branch(Condition=Get DrawToEntireLandscape!(self:Landmass Actor Object Reference=Get SelectedBrush()) -> (DrawToEntireLandscape:bool))
   n4 = Get BrushManager()
   n5 = Break Landmass Landscape Info(LandmassLandscapeInfo=Get LandscapeInformation(self=@n4))
   n6 = Get World Size(self=@n4)
   n7 = Get BrushExtentsMesh()
   > Set World Location!(
    execute
    self:Scene Component Object Reference=@n7
    NewLocation:Vector=vector + vector(A=BreakTransform(InTransform=@n5), B=vector / vector(A=@n6, B=2))
    bSweep:bool=false
    bTeleport:bool=false
   ) -> (then, SweepHitResult:Hit Result Structure) {is=title K2Node_CallFunction}
   > Set World Rotation!(execute, self:Scene Component Object Reference=@n7, NewRotation:Rotator="0, 0, 0", bSweep:bool=false, bTeleport:bool=false) -> (then, SweepHitResult:Hit Result Structure) {is=title K2Node_CallFunction}
   > n8 = SetWorldScale3D!(execute, self:Scene Component Object Reference=@n7, NewScale:Vector=Get World Size(self=Get BrushManager())) -> (then)
   n9 = Get SelectedBrush()
   n10 = Get BrushSize!(self:Landmass Actor Object Reference=@n9) -> (BrushSize:float)
   PrintString(execute=[@n15.then, @n8.then], InString="To String (Float)"(InDouble=@n10))
   n11 = Get ShapeDrawingMesh()
   n12 = BreakVector(InVec=Get World Location!(self:Scene Component Object Reference=@n11) -> (ReturnValue:Vector) {is=title K2Node_CallFunction})
   > Set World Location(self=@n11, NewLocation=MakeVector(X=@n12, Y=@n12.Y, Z=Get Selected Brush Z()), bSweep=false, bTeleport=false)
   > Branch(Condition=Does Object Implement Interface(TestObject=Get SelectedBrush(), Interface=&Landmass_Interface_C))
   n13 = Get BrushExtentsMID()
   > Copy Edge Mask Parameters!(execute:Exec, self:Object Reference=Get SelectedBrush(), MID:Material Instance Dynamic Object Reference=@n13) -> (then:Exec) {is=title K2Node_Message}
   > SetScalarParameterValue(self=@n13, ParameterName=ForceUseRegularUVs, Value=1)
   > SetScalarParameterValue(self=@n13, ParameterName=Brightness, Value=10)
   n14 = Get BrushExtentsMesh()
   Set World Location(execute=@n3.else, self=@n14, NewLocation=Get Actor Location(self=@n9), bSweep=false, bTeleport=false)
   > Set World Rotation(
    self=@n14
    NewRotation=Get Actor Rotation!(self:Actor Object Reference=@n9) -> (ReturnValue:Rotator) {is=title K2Node_CallFunction}
    bSweep=false
    bTeleport=false
   )
   > n15 = SetWorldScale3D(
    self=@n14
    NewScale=vector * vector(A=GetActorScale3D!(self:Actor Object Reference=@n9) -> (ReturnValue:Vector), B:float=@n10) -> (ReturnValue=~)
   )
   Get Target Weightmap() -> (Target Weightmap:Texture Render Target 2D Object Reference)
  ]}
 World to Normalized Landscape Coord {nodes [
   n1 = Get BrushManager()
   n2 = BreakVector2D(InVec=Get RT World Size(self=@n1))
   MakeVector(X=@n2, Y=@n2.Y, Z=1)
   n3 = Get BrushManager()
   n4 = BreakTransform(InTransform=Break Landmass Landscape Info(LandmassLandscapeInfo=Get LandscapeInformation(self=@n3)))
   n5 = Inputs() -> (A=~)
   Outputs(ReturnValue=vector / vector(A=vector - vector(A=@n5, B=@n4), B:Vector=Get World Size(self=@n1)))
   Get RT World Size(self=@n1)
  ]}
]
variables [
 Brushes {type="array<Actor>"}
 BrushExtentsMesh {type=StaticMeshComponent, default=null}
 BrushExtentsMID {type=MaterialInstanceDynamic}
 BrushManager {type=LandmassBrushManager_C, default=null}
 ClickAndDragMode {type=boolean}
 ClickToAddMode {type=boolean}
 CommitShapeRT {type=TextureRenderTarget2D, default=null}
 CopyRTMID {type=MaterialInstanceDynamic}
 CroppedRT {type=TextureRenderTarget2D, default=null}
 DistanceFieldRT {type=TextureRenderTarget2D, default=null}
 Drawing_MouseDragging {type=boolean}
 DrawingMode {type=boolean}
 ExtentsMax {type=Vector2D, default=<0 0>}
 ExtentsMin {type=Vector2D, default=<0 0>}
 Points {type="array<Vector>"}
 PreviewLandscape {type=Landmass_PreviewLandscape_C, default=null}
 Selected Brush Z {type=number, default=1024}
 SelectedBrush {type=LandmassActor, default=null}
 SelectedObject {type=Object, default=null}
 Shape Preview MID {type=MaterialInstanceDynamic}
 ShapeDrawingMesh {type=StaticMeshComponent, default=null}
 ShapeRenderTarget {type=TextureRenderTarget2D, default=null}
 Size {type=number, default=1024}
 World {type=Object, default=null}
]
widgets {
 info {rootWidgetClass=&HorizontalBox, widgetCount=32}
 widgets [
  HorizontalBox_16 {widgetClassPath=&HorizontalBox, widget=HorizontalBox_16}
  VerticalBox_35 {
   widgetClassPath=&VerticalBox
   parent=HorizontalBox_16
   widget=VerticalBox_35
   slot {class=HorizontalBoxSlot, properties {horizontalAlignment=HAlign_Left}}
  }
  SizeBox_46 {
   widgetClassPath=&SizeBox
   parent=VerticalBox_35
   widget=SizeBox_46
   properties {bOverride_WidthOverride=true, widthOverride=400}
   slot {
    class=VerticalBoxSlot
    properties {horizontalAlignment=HAlign_Left, verticalAlignment=VAlign_Top, padding=<ltrb 10 10 10 10>, size {sizeRule=Fill, value=1}}
   }
  }
  EditorUtilityScrollBox_84 {widgetClassPath=&EditorUtilityScrollBox, parent=SizeBox_46, widget=EditorUtilityScrollBox_84, slot {class=SizeBoxSlot}}
  VerticalBox_203 {widgetClassPath=&VerticalBox, parent=EditorUtilityScrollBox_84, widget=VerticalBox_203, slot {class=ScrollBoxSlot}}
  Overlay_123 {widgetClassPath=&Overlay, parent=VerticalBox_203, widget=Overlay_123, slot {class=VerticalBoxSlot}}
  EditorUtilityCombo_AddBrushOptions {
   widgetClassPath=/Script/Blutility.EditorUtilityComboBoxString
   bIsVariable=true
   parent=Overlay_123
   widget=EditorUtilityCombo_AddBrushOptions
   properties {
    selectedOption=A
    defaultOptions [Procedural Noise, Random Flatness, Terraces]
    foregroundColor {colorUseRule=UseColor_Specified, specifiedColor=<rgba 0.5271 0.5271 0.5271 1>}
   }
   slot {class=OverlaySlot, properties {horizontalAlignment=HAlign_Fill, verticalAlignment=VAlign_Fill}}
  }
  EditorUtilityButton_AddBrush {
   widgetClassPath=&EditorUtilityButton
   parent=Overlay_123
   widget=EditorUtilityButton_AddBrush
   properties {visibility=HitTestInvisible}
   slot {class=OverlaySlot, properties {horizontalAlignment=HAlign_Fill, verticalAlignment=VAlign_Fill}}
  }
  TextBlock_2 {
   widgetClassPath=&TextBlock
   parent=EditorUtilityButton_AddBrush
   widget=TextBlock_2
   properties {
    text=Add New Brush Type
    font {
     bForceMonospaced=false
     bMaterialIsStencil=false
     fontMaterial=null
     fontObject=&Roboto
     letterSpacing=0
     monospacedWidth=1
     size=16
     skewAmount=0
     typefaceFontName=Regular
     outlineSettings {
      bApplyOutlineToDropShadows=false
      bMiteredCorners=false
      bSeparateFillAlpha=false
      outlineMaterial=null
      outlineSize=0
      outlineColor=<rgba 0 0 0 1>
     }
    }
   }
   slot {class=ButtonSlot}
  }
  TextBlock_4 {
   widgetClassPath=&TextBlock
   parent=VerticalBox_203
   widget=TextBlock_4
   properties {
    text="Landmass Brushes: "
    font {
     bForceMonospaced=false
     bMaterialIsStencil=false
     fontMaterial=null
     fontObject=&Roboto
     letterSpacing=0
     monospacedWidth=1
     size=8
     skewAmount=0
     typefaceFontName=Regular
     outlineSettings {
      bApplyOutlineToDropShadows=false
      bMiteredCorners=false
      bSeparateFillAlpha=false
      outlineMaterial=null
      outlineSize=0
      outlineColor=<rgba 0 0 0 1>
     }
    }
   }
   slot {class=VerticalBoxSlot, properties {verticalAlignment=VAlign_Center}}
  }
  SizeBox_0 {
   widgetClassPath=&SizeBox
   parent=VerticalBox_203
   widget=SizeBox_0
   properties {bOverride_MaxDesiredHeight=true, maxDesiredHeight=250}
   slot {class=VerticalBoxSlot, properties {verticalAlignment=VAlign_Top, padding=<ltrb 0 10 0 0>}}
  }
  EditorUtilityListView_187 {
   widgetClassPath=/Script/Blutility.EditorUtilityListView
   bIsVariable=true
   parent=SizeBox_0
   widget=EditorUtilityListView_187
   properties {
    entryWidgetClass=&/LandmassBrush_EntryWidget.LandmassBrush_EntryWidget_C
    selectionMode=SingleToggle
   }
   slot {class=SizeBoxSlot, properties {verticalAlignment=VAlign_Top}}
  }
  Details_ItemFromList {
   widgetClassPath=&DetailsView
   parent=VerticalBox_203
   widget=Details_ItemFromList
   properties {bAllowFiltering=true, categoriesToShow [BrushSettings, HeightmapSettings, WeightmapSettings, NoisePasses]}
   slot {class=VerticalBoxSlot}
  }
  VerticalBox_Drawshape {
   widgetClassPath=&VerticalBox
   bIsVariable=true
   parent=VerticalBox_203
   widget=VerticalBox_Drawshape
   properties {visibility=Collapsed}
   slot {class=VerticalBoxSlot}
  }
  EditorUtilityButton_BeginDrawing {
   widgetClassPath=&EditorUtilityButton
   parent=VerticalBox_Drawshape
   widget=EditorUtilityButton_BeginDrawing
   slot {class=VerticalBoxSlot, properties {padding=<ltrb 5 5 5 5>}}
  }
  DrawingButton_Text {
   widgetClassPath=&TextBlock
   bIsVariable=true
   parent=EditorUtilityButton_BeginDrawing
   widget=DrawingButton_Text
   properties {
    text=Begin Drawing Shape
    font {
     bForceMonospaced=false
     bMaterialIsStencil=false
     fontMaterial=null
     fontObject=&Roboto
     letterSpacing=0
     monospacedWidth=1
     size=16
     skewAmount=0
     typefaceFontName=Regular
     outlineSettings {
      bApplyOutlineToDropShadows=false
      bMiteredCorners=false
      bSeparateFillAlpha=false
      outlineMaterial=null
      outlineSize=0
      outlineColor=<rgba 0 0 0 1>
     }
    }
   }
   slot {class=ButtonSlot}
  }
  HorizontalBox_163 {widgetClassPath=&HorizontalBox, parent=VerticalBox_Drawshape, widget=HorizontalBox_163, slot {class=VerticalBoxSlot}}
  TextBlock_1 {
   widgetClassPath=&TextBlock
   parent=HorizontalBox_163
   widget=TextBlock_1
   properties {
    text=Drawing Mode
    font {
     bForceMonospaced=false
     bMaterialIsStencil=false
     fontMaterial=null
     fontObject=&Roboto
     letterSpacing=0
     monospacedWidth=1
     size=8
     skewAmount=0
     typefaceFontName=Regular
     outlineSettings {
      bApplyOutlineToDropShadows=false
      bMiteredCorners=false
      bSeparateFillAlpha=false
      outlineMaterial=null
      outlineSize=0
      outlineColor=<rgba 0 0 0 1>
     }
    }
   }
   slot {class=HorizontalBoxSlot, properties {verticalAlignment=VAlign_Center}}
  }
  EditorUtilityComboBoxString_ClickDragMode {
   widgetClassPath=&EditorUtilityComboBoxString
   bIsVariable=true
   parent=HorizontalBox_163
   widget=EditorUtilityComboBoxString_ClickDragMode
   properties {
    selectedOption=Click and Drag
    defaultOptions [Click and Drag, Click to Add Points]
    font {
     bForceMonospaced=false
     bMaterialIsStencil=false
     fontMaterial=null
     fontObject=&Roboto
     letterSpacing=0
     monospacedWidth=1
     size=10
     skewAmount=0
     typefaceFontName=Regular
     outlineSettings {
      bApplyOutlineToDropShadows=false
      bMiteredCorners=false
      bSeparateFillAlpha=false
      outlineMaterial=null
      outlineSize=0
      outlineColor=<rgba 0 0 0 1>
     }
    }
    foregroundColor {colorUseRule=UseColor_Specified, specifiedColor=<rgba 0.5271 0.5271 0.5271 1>}
   }
   slot {class=HorizontalBoxSlot, properties {padding=<ltrb 5 5 5 5>, size {sizeRule=Fill, value=1}}}
  }
  HorizontalBox_98 {widgetClassPath=&HorizontalBox, parent=VerticalBox_Drawshape, widget=HorizontalBox_98, slot {class=VerticalBoxSlot}}
  TextBlock_84 {
   widgetClassPath=&TextBlock
   parent=HorizontalBox_98
   widget=TextBlock_84
   properties {
    text="Blur Steps         "
    font {
     bForceMonospaced=false
     bMaterialIsStencil=false
     fontMaterial=null
     fontObject=&Roboto
     letterSpacing=0
     monospacedWidth=1
     size=8
     skewAmount=0
     typefaceFontName=Regular
     outlineSettings {
      bApplyOutlineToDropShadows=false
      bMiteredCorners=false
      bSeparateFillAlpha=false
      outlineMaterial=null
      outlineSize=0
      outlineColor=<rgba 0 0 0 1>
     }
    }
   }
   slot {class=HorizontalBoxSlot, properties {verticalAlignment=VAlign_Center}}
  }
  EditorUtilityEditableTextBox_BlurSteps {
   widgetClassPath=&EditorUtilityEditableTextBox
   bIsVariable=true
   parent=HorizontalBox_98
   widget=EditorUtilityEditableTextBox_BlurSteps
   properties {
    text=8
    widgetStyle {
     backgroundColor {colorUseRule=UseColor_Specified, specifiedColor=<rgba 1 1 1 1>}
     backgroundImageFocused {
      bIsDynamicallyLoaded=false
      drawAs=RoundedBox
      imageType=NoImage
      mirroring=NoMirror
      resourceName=None
      resourceObject=null
      tiling=NoTile
      imageSize=<0 0>
      margin=<ltrb 0 0 0 0>
      outlineSettings {
       bUseBrushTransparency=false
       roundingType=FixedRadius
       width=1
       color {colorUseRule=UseColor_ColorTable, specifiedColor=<rgba 0 0 0 0>}
       cornerRadii=<4 4 4 4>
      }
      tintColor {colorUseRule=UseColor_Specified, specifiedColor=<rgba 0.0048 0.0048 0.0048 1>}
      uVRegion {bIsValid=false, max=<0 0>, min=<0 0>}
     }
     backgroundImageHovered {
      bIsDynamicallyLoaded=false
      drawAs=RoundedBox
      imageType=NoImage
      mirroring=NoMirror
      resourceName=None
      resourceObject=null
      tiling=NoTile
      imageSize=<0 0>
      margin=<ltrb 0 0 0 0>
      outlineSettings {
       bUseBrushTransparency=false
       roundingType=FixedRadius
       width=1
       color {colorUseRule=UseColor_ColorTable, specifiedColor=<rgba 0 0 0 0>}
       cornerRadii=<4 4 4 4>
      }
      tintColor {colorUseRule=UseColor_Specified, specifiedColor=<rgba 0.0048 0.0048 0.0048 1>}
      uVRegion {bIsValid=false, max=<0 0>, min=<0 0>}
     }
     backgroundImageNormal {
      bIsDynamicallyLoaded=false
      drawAs=RoundedBox
      imageType=NoImage
      mirroring=NoMirror
      resourceName=None
      resourceObject=null
      tiling=NoTile
      imageSize=<0 0>
      margin=<ltrb 0 0 0 0>
      outlineSettings {
       bUseBrushTransparency=false
       roundingType=FixedRadius
       width=1
       color {colorUseRule=UseColor_ColorTable, specifiedColor=<rgba 0 0 0 0>}
       cornerRadii=<4 4 4 4>
      }
      tintColor {colorUseRule=UseColor_Specified, specifiedColor=<rgba 0.0048 0.0048 0.0048 1>}
      uVRegion {bIsValid=false, max=<0 0>, min=<0 0>}
     }
     backgroundImageReadOnly {
      bIsDynamicallyLoaded=false
      drawAs=RoundedBox
      imageType=NoImage
      mirroring=NoMirror
      resourceName=None
      resourceObject=null
      tiling=NoTile
      imageSize=<0 0>
      margin=<ltrb 0 0 0 0>
      outlineSettings {
       bUseBrushTransparency=false
       roundingType=FixedRadius
       width=1
       color {colorUseRule=UseColor_ColorTable, specifiedColor=<rgba 0 0 0 0>}
       cornerRadii=<4 4 4 4>
      }
      tintColor {colorUseRule=UseColor_Specified, specifiedColor=<rgba 0.0284 0.0284 0.0284 1>}
      uVRegion {bIsValid=false, max=<0 0>, min=<0 0>}
     }
     focusedForegroundColor {colorUseRule=UseColor_Specified, specifiedColor=<rgba 1 1 1 1>}
     foregroundColor {colorUseRule=UseColor_Specified, specifiedColor=<rgba 0.5271 0.5271 0.5271 1>}
     hScrollBarPadding=<ltrb 0 0 0 0>
     padding=<ltrb 8 3 8 4>
     readOnlyForegroundColor {colorUseRule=UseColor_Foreground, specifiedColor=<rgba 1 0 1 1>}
     scrollBarStyle {
      thickness=8
      draggedThumbImage {
       bIsDynamicallyLoaded=false
       drawAs=RoundedBox
       imageType=NoImage
       mirroring=NoMirror
       resourceName=None
       resourceObject=null
       tiling=NoTile
       imageSize=<0 0>
       margin=<ltrb 0 0 0 0>
       outlineSettings {
        bUseBrushTransparency=false
        roundingType=FixedRadius
        width=0
        color {colorUseRule=UseColor_Specified, specifiedColor=<rgba 0 0 0 0>}
        cornerRadii=<4 4 4 4>
       }
       tintColor {colorUseRule=UseColor_Specified, specifiedColor=<rgba 0.2159 0.2159 0.2159 1>}
       uVRegion {bIsValid=false, max=<0 0>, min=<0 0>}
      }
      horizontalBackgroundImage {
       bIsDynamicallyLoaded=false
       drawAs=NoDrawType
       imageType=NoImage
       mirroring=NoMirror
       resourceName=None
       resourceObject=null
       tiling=NoTile
       imageSize=<0 0>
       margin=<ltrb 0 0 0 0>
       outlineSettings {
        bUseBrushTransparency=false
        roundingType=HalfHeightRadius
        width=0
        color {colorUseRule=UseColor_Specified, specifiedColor=<rgba 0 0 0 0>}
        cornerRadii=<0 0 0 0>
       }
       tintColor {colorUseRule=UseColor_Specified, specifiedColor=<rgba 1 1 1 1>}
       uVRegion {bIsValid=false, max=<0 0>, min=<0 0>}
      }
      horizontalBottomSlotImage {
       bIsDynamicallyLoaded=false
       drawAs=NoDrawType
       imageType=NoImage
       mirroring=NoMirror
       resourceName=None
       resourceObject=null
       tiling=NoTile
       imageSize=<0 0>
       margin=<ltrb 0 0 0 0>
       outlineSettings {
        bUseBrushTransparency=false
        roundingType=HalfHeightRadius
        width=0
        color {colorUseRule=UseColor_Specified, specifiedColor=<rgba 0 0 0 0>}
        cornerRadii=<0 0 0 0>
       }
       tintColor {colorUseRule=UseColor_Specified, specifiedColor=<rgba 1 1 1 1>}
       uVRegion {bIsValid=false, max=<0 0>, min=<0 0>}
      }
      horizontalTopSlotImage {
       bIsDynamicallyLoaded=false
       drawAs=NoDrawType
       imageType=NoImage
       mirroring=NoMirror
       resourceName=None
       resourceObject=null
       tiling=NoTile
       imageSize=<0 0>
       margin=<ltrb 0 0 0 0>
       outlineSettings {
        bUseBrushTransparency=false
        roundingType=HalfHeightRadius
        width=0
        color {colorUseRule=UseColor_Specified, specifiedColor=<rgba 0 0 0 0>}
        cornerRadii=<0 0 0 0>
       }
       tintColor {colorUseRule=UseColor_Specified, specifiedColor=<rgba 1 1 1 1>}
       uVRegion {bIsValid=false, max=<0 0>, min=<0 0>}
      }
      hoveredThumbImage {
       bIsDynamicallyLoaded=false
       drawAs=RoundedBox
       imageType=NoImage
       mirroring=NoMirror
       resourceName=None
       resourceObject=null
       tiling=NoTile
       imageSize=<0 0>
       margin=<ltrb 0 0 0 0>
       outlineSettings {
        bUseBrushTransparency=false
        roundingType=FixedRadius
        width=0
        color {colorUseRule=UseColor_Specified, specifiedColor=<rgba 0 0 0 0>}
        cornerRadii=<4 4 4 4>
       }
       tintColor {colorUseRule=UseColor_Specified, specifiedColor=<rgba 0.2159 0.2159 0.2159 1>}
       uVRegion {bIsValid=false, max=<0 0>, min=<0 0>}
      }
      normalThumbImage {
       bIsDynamicallyLoaded=false
       drawAs=RoundedBox
       imageType=NoImage
       mirroring=NoMirror
       resourceName=None
       resourceObject=null
       tiling=NoTile
       imageSize=<0 0>
       margin=<ltrb 0 0 0 0>
       outlineSettings {
        bUseBrushTransparency=false
        roundingType=FixedRadius
        width=0
        color {colorUseRule=UseColor_Specified, specifiedColor=<rgba 0 0 0 0>}
        cornerRadii=<4 4 4 4>
       }
       tintColor {colorUseRule=UseColor_Specified, specifiedColor=<rgba 0.0953 0.0953 0.0953 1>}
       uVRegion {bIsValid=false, max=<0 0>, min=<0 0>}
      }
      verticalBackgroundImage {
       bIsDynamicallyLoaded=false
       drawAs=NoDrawType
       imageType=NoImage
       mirroring=NoMirror
       resourceName=None
       resourceObject=null
       tiling=NoTile
       imageSize=<0 0>
       margin=<ltrb 0 0 0 0>
       outlineSettings {
        bUseBrushTransparency=false
        roundingType=HalfHeightRadius
        width=0
        color {colorUseRule=UseColor_Specified, specifiedColor=<rgba 0 0 0 0>}
        cornerRadii=<0 0 0 0>
       }
       tintColor {colorUseRule=UseColor_Specified, specifiedColor=<rgba 1 1 1 1>}
       uVRegion {bIsValid=false, max=<0 0>, min=<0 0>}
      }
      verticalBottomSlotImage {
       bIsDynamicallyLoaded=false
       drawAs=NoDrawType
       imageType=NoImage
       mirroring=NoMirror
       resourceName=None
       resourceObject=null
       tiling=NoTile
       imageSize=<0 0>
       margin=<ltrb 0 0 0 0>
       outlineSettings {
        bUseBrushTransparency=false
        roundingType=HalfHeightRadius
        width=0
        color {colorUseRule=UseColor_Specified, specifiedColor=<rgba 0 0 0 0>}
        cornerRadii=<0 0 0 0>
       }
       tintColor {colorUseRule=UseColor_Specified, specifiedColor=<rgba 1 1 1 1>}
       uVRegion {bIsValid=false, max=<0 0>, min=<0 0>}
      }
      verticalTopSlotImage {
       bIsDynamicallyLoaded=false
       drawAs=NoDrawType
       imageType=NoImage
       mirroring=NoMirror
       resourceName=None
       resourceObject=null
       tiling=NoTile
       imageSize=<0 0>
       margin=<ltrb 0 0 0 0>
       outlineSettings {
        bUseBrushTransparency=false
        roundingType=HalfHeightRadius
        width=0
        color {colorUseRule=UseColor_Specified, specifiedColor=<rgba 0 0 0 0>}
        cornerRadii=<0 0 0 0>
       }
       tintColor {colorUseRule=UseColor_Specified, specifiedColor=<rgba 1 1 1 1>}
       uVRegion {bIsValid=false, max=<0 0>, min=<0 0>}
      }
     }
     textStyle {
      overflowPolicy=Clip
      transformPolicy=None
      colorAndOpacity {colorUseRule=UseColor_Foreground, specifiedColor=<rgba 1 0 1 1>}
      font {
       bForceMonospaced=false
       bMaterialIsStencil=false
       fontMaterial=null
       fontObject=&Roboto
       letterSpacing=0
       monospacedWidth=1
       size=10
       skewAmount=0
       typefaceFontName=Regular
       outlineSettings {
        bApplyOutlineToDropShadows=false
        bMiteredCorners=false
        bSeparateFillAlpha=false
        outlineMaterial=null
        outlineSize=0
        outlineColor=<rgba 0 0 0 1>
       }
      }
      highlightColor {colorUseRule=UseColor_ColorTable, specifiedColor=<rgba 0 0 0 0>}
      highlightShape {
       bIsDynamicallyLoaded=false
       drawAs=Image
       imageType=NoImage
       mirroring=NoMirror
       resourceName=None
       resourceObject=null
       tiling=NoTile
       imageSize=<0 0>
       margin=<ltrb 0 0 0 0>
       outlineSettings {
        bUseBrushTransparency=false
        roundingType=HalfHeightRadius
        width=0
        color {colorUseRule=UseColor_Specified, specifiedColor=<rgba 0 0 0 0>}
        cornerRadii=<0 0 0 0>
       }
       tintColor {colorUseRule=UseColor_ColorTable, specifiedColor=<rgba 0 0 0 0>}
       uVRegion {bIsValid=false, max=<0 0>, min=<0 0>}
      }
      selectedBackgroundColor {colorUseRule=UseColor_ColorTable, specifiedColor=<rgba 0 0 0 0>}
      shadowColorAndOpacity=<rgba 0 0 0 1>
      shadowOffset=<0 0>
      strikeBrush {
       bIsDynamicallyLoaded=false
       drawAs=Image
       imageType=NoImage
       mirroring=NoMirror
       resourceName=None
       resourceObject=null
       tiling=NoTile
       imageSize=<32 32>
       margin=<ltrb 0 0 0 0>
       outlineSettings {
        bUseBrushTransparency=false
        roundingType=HalfHeightRadius
        width=0
        color {colorUseRule=UseColor_Specified, specifiedColor=<rgba 0 0 0 0>}
        cornerRadii=<0 0 0 0>
       }
       tintColor {colorUseRule=UseColor_Specified, specifiedColor=<rgba 1 1 1 1>}
       uVRegion {bIsValid=false, max=<0 0>, min=<0 0>}
      }
      underlineBrush {
       bIsDynamicallyLoaded=false
       drawAs=Image
       imageType=NoImage
       mirroring=NoMirror
       resourceName=None
       resourceObject=null
       tiling=NoTile
       imageSize=<32 32>
       margin=<ltrb 0 0 0 0>
       outlineSettings {
        bUseBrushTransparency=false
        roundingType=HalfHeightRadius
        width=0
        color {colorUseRule=UseColor_Specified, specifiedColor=<rgba 0 0 0 0>}
        cornerRadii=<0 0 0 0>
       }
       tintColor {colorUseRule=UseColor_Specified, specifiedColor=<rgba 1 1 1 1>}
       uVRegion {bIsValid=false, max=<0 0>, min=<0 0>}
      }
     }
     vScrollBarPadding=<ltrb 0 0 0 0>
    }
   }
   slot {class=HorizontalBoxSlot, properties {padding=<ltrb 1 0 0 0>}}
  }
  EditorUtilityButton_MakeDF {
   widgetClassPath=&EditorUtilityButton
   parent=VerticalBox_Drawshape
   widget=EditorUtilityButton_MakeDF
   properties {visibility=Collapsed}
   slot {class=VerticalBoxSlot, properties {padding=<ltrb 5 5 5 5>}}
  }
  TextBlock {
   widgetClassPath=&TextBlock
   parent=EditorUtilityButton_MakeDF
   widget=TextBlock
   properties {
    text=Make Distance Field
    font {
     bForceMonospaced=false
     bMaterialIsStencil=false
     fontMaterial=null
     fontObject=&Roboto
     letterSpacing=0
     monospacedWidth=1
     size=16
     skewAmount=0
     typefaceFontName=Regular
     outlineSettings {
      bApplyOutlineToDropShadows=false
      bMiteredCorners=false
      bSeparateFillAlpha=false
      outlineMaterial=null
      outlineSize=0
      outlineColor=<rgba 0 0 0 1>
     }
    }
   }
   slot {class=ButtonSlot}
  }
  Button_SaveShapeToBrush {
   widgetClassPath=&EditorUtilityButton
   parent=VerticalBox_Drawshape
   widget=Button_SaveShapeToBrush
   properties {visibility=Collapsed}
   slot {class=VerticalBoxSlot, properties {padding=<ltrb 5 5 5 5>}}
  }
  TextBlock_3 {
   widgetClassPath=&TextBlock
   parent=Button_SaveShapeToBrush
   widget=TextBlock_3
   properties {
    text=Save to Selected Brush
    font {
     bForceMonospaced=false
     bMaterialIsStencil=false
     fontMaterial=null
     fontObject=&Roboto
     letterSpacing=0
     monospacedWidth=1
     size=16
     skewAmount=0
     typefaceFontName=Regular
     outlineSettings {
      bApplyOutlineToDropShadows=false
      bMiteredCorners=false
      bSeparateFillAlpha=false
      outlineMaterial=null
      outlineSize=0
      outlineColor=<rgba 0 0 0 1>
     }
    }
   }
   slot {class=ButtonSlot}
  }
  SinglePropertyView_ShapeRT {
   widgetClassPath=&SinglePropertyView
   parent=VerticalBox_Drawshape
   widget=SinglePropertyView_ShapeRT
   properties {propertyName=ShapeRenderTarget, visibility=Collapsed}
   slot {class=VerticalBoxSlot}
  }
  SinglePropertyView_CroppedRT {
   widgetClassPath=&SinglePropertyView
   parent=VerticalBox_Drawshape
   widget=SinglePropertyView_CroppedRT
   properties {propertyName=CroppedRT, visibility=Collapsed}
   slot {class=VerticalBoxSlot}
  }
  SinglePropertyView_DFRT {
   widgetClassPath=&SinglePropertyView
   parent=VerticalBox_Drawshape
   widget=SinglePropertyView_DFRT
   properties {propertyName=BrushExtentsMID, visibility=Collapsed}
   slot {class=VerticalBoxSlot}
  }
  Overlay_27 {
   widgetClassPath=&Overlay
   parent=HorizontalBox_16
   widget=Overlay_27
   slot {class=HorizontalBoxSlot, properties {size {sizeRule=Fill, value=1}}}
  }
  LandmassViewport_Widget {
   widgetClassPath=&/LandmassViewport_Widget
   bIsVariable=true
   parent=Overlay_27
   widget=LandmassViewport_Widget
   slot {class=OverlaySlot, properties {horizontalAlignment=HAlign_Fill, verticalAlignment=VAlign_Fill}}
  }
  MouseCapture_BorderWidget {
   widgetClassPath=&Border
   bIsVariable=true
   parent=Overlay_27
   widget=MouseCapture_BorderWidget
   properties {renderOpacity=0.1, visibility=Collapsed}
   slot {class=OverlaySlot, properties {horizontalAlignment=HAlign_Fill, verticalAlignment=VAlign_Fill}}
  }
 ]
}
