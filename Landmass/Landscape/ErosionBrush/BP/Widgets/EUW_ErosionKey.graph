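# Editor utility widget for one erosion key in the Landmass Landscape Erosion brush tool, shown as a button with a name label.
# It stores an erosion mode plus settings for particle erosion, water levelling, slope smoothing, noise injection and a custom material pass.
# Clicking the button passes itself to the parent EUW_Landscape_Erosion widget through Clicked Erosion Key.
# Default and selected button styles are defined as variables, but this graph never applies them.
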EditorUtilityWidgetBlueprint /Landmass/Landscape/ErosionBrush/BP/Widgets/EUW_ErosionKey
graphs [EventGraph {nodes [
   "On Clicked (Button_0)"() {is=title K2Node_ComponentBoundEvent}
   > Clicked Erosion Key!(
    execute
    self:EUW Landscape Erosion Object Reference=Get ParentWidget() -> (ParentWidget:EUW Landscape Erosion Object Reference)
    Key:EUW Erosion Key Object Reference=Self-Reference()
   ) -> (then)
   Event PreConstruct()
   Event Construct()
   Event Tick()
  ]}]
variables [
 Custom Material Pass {
  type=Landmass_Erosion_CustomMaterial
  default {iterations=100, material=null, scalar Parameters {}, texture Parameters {}, vector Parameters {}}
 }
 Data Asset {type=boolean}
 Default_ButtonStyle {
  type=ButtonStyle
  default {
   clickedSlateSound {resourceObject=null}
   disabled {
    bIsDynamicallyLoaded=false
    drawAs=RoundedBox
    imageType=NoImage
    mirroring=NoMirror
    resourceName=None
    resourceObject=null
    tiling=NoTile
    imageSize=<0 0>
    margin=<ltrb 0 0 0 0>
    outlineSettings {
     bUseBrushTransparency=false
     roundingType=FixedRadius
     width=1
     color {colorUseRule=UseColor_ColorTable, specifiedColor=<rgba 0 0 0 0>}
     cornerRadii=<4 4 4 4>
    }
    tintColor {colorUseRule=UseColor_Specified, specifiedColor=<rgba 0.0395 0.0395 0.0395 1>}
    uVRegion {bIsValid=false, max=<0 0>, min=<0 0>}
   }
   disabledForeground {colorUseRule=UseColor_ColorTable, specifiedColor=<rgba 0 0 0 0>}
   hovered {
    bIsDynamicallyLoaded=false
    drawAs=RoundedBox
    imageType=NoImage
    mirroring=NoMirror
    resourceName=None
    resourceObject=null
    tiling=NoTile
    imageSize=<32 32>
    margin=<ltrb 0 0 0 0>
    outlineSettings {
     bUseBrushTransparency=false
     roundingType=FixedRadius
     width=1
     color {colorUseRule=UseColor_ColorTable, specifiedColor=<rgba 0 0 0 0>}
     cornerRadii=<4 4 4 4>
    }
    tintColor {colorUseRule=UseColor_Specified, specifiedColor=<rgba 0.0953 0.0953 0.0953 1>}
    uVRegion {bIsValid=false, max=<0 0>, min=<0 0>}
   }
   hoveredForeground {colorUseRule=UseColor_ColorTable, specifiedColor=<rgba 0 0 0 0>}
   hoveredSlateSound {resourceObject=null}
   normal {
    bIsDynamicallyLoaded=false
    drawAs=RoundedBox
    imageType=NoImage
    mirroring=NoMirror
    resourceName=None
    resourceObject=null
    tiling=NoTile
    imageSize=<32 32>
    margin=<ltrb 0 0 0 0>
    outlineSettings {
     bUseBrushTransparency=false
     roundingType=FixedRadius
     width=1
     color {colorUseRule=UseColor_ColorTable, specifiedColor=<rgba 0 0 0 0>}
     cornerRadii=<4 4 4 4>
    }
    tintColor {colorUseRule=UseColor_Specified, specifiedColor=<rgba 0.0395 0.0395 0.0395 1>}
    uVRegion {bIsValid=false, max=<0 0>, min=<0 0>}
   }
   normalForeground {colorUseRule=UseColor_ColorTable, specifiedColor=<rgba 0 0 0 0>}
   normalPadding=<ltrb 12 1.5 12 1.5>
   pressed {
    bIsDynamicallyLoaded=false
    drawAs=RoundedBox
    imageType=NoImage
    mirroring=NoMirror
    resourceName=None
    resourceObject=null
    tiling=NoTile
    imageSize=<32 32>
    margin=<ltrb 0 0 0 0>
    outlineSettings {
     bUseBrushTransparency=false
     roundingType=FixedRadius
     width=1
     color {colorUseRule=UseColor_ColorTable, specifiedColor=<rgba 0 0 0 0>}
     cornerRadii=<4 4 4 4>
    }
    tintColor {colorUseRule=UseColor_Specified, specifiedColor=<rgba 0.0284 0.0284 0.0284 1>}
    uVRegion {bIsValid=false, max=<0 0>, min=<0 0>}
   }
   pressedForeground {colorUseRule=UseColor_ColorTable, specifiedColor=<rgba 0 0 0 0>}
   pressedPadding=<ltrb 12 2.5 12 0.5>
   pressedSlateSound {resourceObject=null}
  }
 }
 Erosion Mode {type=E_Landmass_CustomBrush_ErosionModes, default=Particle Erosion}
 Key Name {type=string, default=Erosion Key}
 Noise Injection Settings {
  type=Landmass_Erosion_NoiseInjection
  default {
   noise Floor 0-1=0.3
   noise Height Meters=4
   noise Octaves=6
   noise Persistence=0.5
   noise Scale Meters=100
   noise Seed=0
   turbulence=false
   slopeMask {enable Slope Mask=false, max Slope=1, min Slope=0.25, slope Falloff=0.1}
  }
 }
 ParentWidget {type=EUW_Landscape_Erosion_C, default=null}
 Particle Erosion Settings {
  type=Landmass_Particle_Erosion_Settings
  default {
   delta T=0.0167
   erosion Iterations=1000
   evaporation=0
   reset Particles Every N Iterations=0
   resolution Divisor=1
   solubility=0
   starting Water Amount=100
   velocity Advection=0
   velocity Damping=0.2
   erode Weightmaps []
  }
 }
 Selected_ButtonStyle {
  type=ButtonStyle
  default {
   clickedSlateSound {resourceObject=null}
   disabled {
    bIsDynamicallyLoaded=false
    drawAs=RoundedBox
    imageType=NoImage
    mirroring=NoMirror
    resourceName=None
    resourceObject=null
    tiling=NoTile
    imageSize=<0 0>
    margin=<ltrb 0 0 0 0>
    outlineSettings {
     bUseBrushTransparency=false
     roundingType=FixedRadius
     width=1
     color {colorUseRule=UseColor_ColorTable, specifiedColor=<rgba 0 0 0 0>}
     cornerRadii=<4 4 4 4>
    }
    tintColor {colorUseRule=UseColor_Specified, specifiedColor=<rgba 0.0395 0.0395 0.0395 1>}
    uVRegion {bIsValid=false, max=<0 0>, min=<0 0>}
   }
   disabledForeground {colorUseRule=UseColor_ColorTable, specifiedColor=<rgba 0 0 0 0>}
   hovered {
    bIsDynamicallyLoaded=false
    drawAs=RoundedBox
    imageType=NoImage
    mirroring=NoMirror
    resourceName=None
    resourceObject=null
    tiling=NoTile
    imageSize=<32 32>
    margin=<ltrb 0 0 0 0>
    outlineSettings {
     bUseBrushTransparency=false
     roundingType=FixedRadius
     width=1
     color {colorUseRule=UseColor_ColorTable, specifiedColor=<rgba 0 0 0 0>}
     cornerRadii=<4 4 4 4>
    }
    tintColor {colorUseRule=UseColor_Specified, specifiedColor=<rgba 0.0256 0.2379 0.9271 1>}
    uVRegion {bIsValid=false, max=<0 0>, min=<0 0>}
   }
   hoveredForeground {colorUseRule=UseColor_ColorTable, specifiedColor=<rgba 0 0 0 0>}
   hoveredSlateSound {resourceObject=null}
   normal {
    bIsDynamicallyLoaded=false
    drawAs=RoundedBox
    imageType=NoImage
    mirroring=NoMirror
    resourceName=None
    resourceObject=null
    tiling=NoTile
    imageSize=<32 32>
    margin=<ltrb 0 0 0 0>
    outlineSettings {
     bUseBrushTransparency=false
     roundingType=FixedRadius
     width=1
     color {colorUseRule=UseColor_ColorTable, specifiedColor=<rgba 0 0 0 0>}
     cornerRadii=<4 4 4 4>
    }
    tintColor {colorUseRule=UseColor_Specified, specifiedColor=<rgba 0.0256 0.2379 0.9271 1>}
    uVRegion {bIsValid=false, max=<0 0>, min=<0 0>}
   }
   normalForeground {colorUseRule=UseColor_ColorTable, specifiedColor=<rgba 0 0 0 0>}
   normalPadding=<ltrb 12 1.5 12 1.5>
   pressed {
    bIsDynamicallyLoaded=false
    drawAs=RoundedBox
    imageType=NoImage
    mirroring=NoMirror
    resourceName=None
    resourceObject=null
    tiling=NoTile
    imageSize=<32 32>
    margin=<ltrb 0 0 0 0>
    outlineSettings {
     bUseBrushTransparency=false
     roundingType=FixedRadius
     width=1
     color {colorUseRule=UseColor_ColorTable, specifiedColor=<rgba 0 0 0 0>}
     cornerRadii=<4 4 4 4>
    }
    tintColor {colorUseRule=UseColor_Specified, specifiedColor=<rgba 0.0284 0.0284 0.0284 1>}
    uVRegion {bIsValid=false, max=<0 0>, min=<0 0>}
   }
   pressedForeground {colorUseRule=UseColor_ColorTable, specifiedColor=<rgba 0 0 0 0>}
   pressedPadding=<ltrb 12 2.5 12 0.5>
   pressedSlateSound {resourceObject=null}
  }
 }
 Slope Smoothing Settings {type=Landmass_SlopeSmoothing_Settings, default {iterations=100, opacity=1, smooth Slopes Greater Than=0, smooth Slopes Less Than=0}}
 Water Levelling Settings {
  type=Landmass_WaterLevelling_Settings
  default {
   evaporation=0.1
   iterations=100
   rain Per Iteration=0.0167
   reset Water Every N Iterations=0
   solubility=0.1
   starting Water Amount=10
   erode Weightmaps []
  }
 }
]
widgets {
 info {rootWidgetClass=&Button, widgetCount=4}
 widgets [
  Button_0 {
   widgetClassPath=&Button
   widget=Button_0
   properties {widgetStyle {
     clickedSlateSound {resourceObject=null}
     disabled {
      bIsDynamicallyLoaded=false
      drawAs=RoundedBox
      imageType=NoImage
      mirroring=NoMirror
      resourceName=None
      resourceObject=null
      tiling=NoTile
      imageSize=<0 0>
      margin=<ltrb 0 0 0 0>
      outlineSettings {
       bUseBrushTransparency=false
       roundingType=FixedRadius
       width=1
       color {colorUseRule=UseColor_ColorTable, specifiedColor=<rgba 0 0 0 0>}
       cornerRadii=<4 4 4 4>
      }
      tintColor {colorUseRule=UseColor_Specified, specifiedColor=<rgba 0.0395 0.0395 0.0395 1>}
      uVRegion {bIsValid=false, max=<0 0>, min=<0 0>}
     }
     disabledForeground {colorUseRule=UseColor_ColorTable, specifiedColor=<rgba 0 0 0 0>}
     hovered {
      bIsDynamicallyLoaded=false
      drawAs=RoundedBox
      imageType=NoImage
      mirroring=NoMirror
      resourceName=None
      resourceObject=null
      tiling=NoTile
      imageSize=<32 32>
      margin=<ltrb 0 0 0 0>
      outlineSettings {
       bUseBrushTransparency=false
       roundingType=FixedRadius
       width=1
       color {colorUseRule=UseColor_ColorTable, specifiedColor=<rgba 0 0 0 0>}
       cornerRadii=<4 4 4 4>
      }
      tintColor {colorUseRule=UseColor_Specified, specifiedColor=<rgba 0.0953 0.0953 0.0953 1>}
      uVRegion {bIsValid=false, max=<0 0>, min=<0 0>}
     }
     hoveredForeground {colorUseRule=UseColor_ColorTable, specifiedColor=<rgba 0 0 0 0>}
     hoveredSlateSound {resourceObject=null}
     normal {
      bIsDynamicallyLoaded=false
      drawAs=RoundedBox
      imageType=NoImage
      mirroring=NoMirror
      resourceName=None
      resourceObject=null
      tiling=NoTile
      imageSize=<32 32>
      margin=<ltrb 0 0 0 0>
      outlineSettings {
       bUseBrushTransparency=false
       roundingType=FixedRadius
       width=1
       color {colorUseRule=UseColor_ColorTable, specifiedColor=<rgba 0 0 0 0>}
       cornerRadii=<4 4 4 4>
      }
      tintColor {colorUseRule=UseColor_Specified, specifiedColor=<rgba 0.0395 0.0395 0.0395 1>}
      uVRegion {bIsValid=false, max=<0 0>, min=<0 0>}
     }
     normalForeground {colorUseRule=UseColor_ColorTable, specifiedColor=<rgba 0 0 0 0>}
     normalPadding=<ltrb 12 1.5 12 1.5>
     pressed {
      bIsDynamicallyLoaded=false
      drawAs=RoundedBox
      imageType=NoImage
      mirroring=NoMirror
      resourceName=None
      resourceObject=null
      tiling=NoTile
      imageSize=<32 32>
      margin=<ltrb 0 0 0 0>
      outlineSettings {
       bUseBrushTransparency=false
       roundingType=FixedRadius
       width=1
       color {colorUseRule=UseColor_ColorTable, specifiedColor=<rgba 0 0 0 0>}
       cornerRadii=<4 4 4 4>
      }
      tintColor {colorUseRule=UseColor_Specified, specifiedColor=<rgba 0.0284 0.0284 0.0284 1>}
      uVRegion {bIsValid=false, max=<0 0>, min=<0 0>}
     }
     pressedForeground {colorUseRule=UseColor_ColorTable, specifiedColor=<rgba 0 0 0 0>}
     pressedPadding=<ltrb 12 2.5 12 0.5>
     pressedSlateSound {resourceObject=null}
    }}
  }
  VerticalBox_311 {widgetClassPath=&VerticalBox, bIsVariable=true, parent=Button_0, widget=VerticalBox_311, slot {class=ButtonSlot}}
  TextBlock_KeyName {
   widgetClassPath=&TextBlock
   bIsVariable=true
   parent=VerticalBox_311
   widget=TextBlock_KeyName
   properties {
    text=Erosion Key
    colorAndOpacity {colorUseRule=UseColor_Specified, specifiedColor=<rgba 0.6615 0.6615 0.6615 1>}
    font {
     bForceMonospaced=false
     bMaterialIsStencil=false
     fontMaterial=null
     fontObject=&Roboto
     letterSpacing=0
     monospacedWidth=1
     size=10
     skewAmount=0
     typefaceFontName=Regular
     outlineSettings {
      bApplyOutlineToDropShadows=false
      bMiteredCorners=false
      bSeparateFillAlpha=false
      outlineMaterial=null
      outlineSize=0
      outlineColor=<rgba 0 0 0 1>
     }
    }
   }
   slot {class=VerticalBoxSlot, properties {horizontalAlignment=HAlign_Center}}
  }
  Overlay_112 {widgetClassPath=&Overlay, parent=VerticalBox_311, widget=Overlay_112, slot {class=VerticalBoxSlot}}
 ]
}
