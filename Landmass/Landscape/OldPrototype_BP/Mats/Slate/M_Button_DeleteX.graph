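# Translucent unlit UI material for a Slate button that draws a procedural X delete icon in the old landscape Blueprint prototype.
# Two DrawLine-2D calls share endpoint, radius and falloff parameters to draw a diagonal stroke.
# One call receives UVs with U flipped to mirror the stroke, and a Max merges both into the cross without any texture.
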Material /Landmass/Landscape/OldPrototype_BP/Mats/Slate/M_Button_DeleteX
graph {
 properties {blendMode=BLEND_Translucent, materialDomain=MD_UI, pixelDepthOffsetMode=PDOM_Legacy, shadingModel=MSM_Unlit, usedShadingModels=Unlit}
 nodes [
  Constant() {r=1}
  n1 = MaterialFunctionCall(Float2=TextureCoordinate()) {materialFunction=&BreakOutFloat2Components}
  n2 = OneMinus(None=@n1.R)
  Max(
   A=MaterialFunctionCall!(
    Coordinates=AppendVector(A=@n2, B=@n1.G)
    P0=VectorParameter() {parameterName=A, defaultValue=<rgba 0.25 0.75 0 0>}.RGB
    P1=VectorParameter() {parameterName=B, defaultValue=<rgba 0.75 0.25 0 0>}.RGB
    Line Width=ScalarParameter() {defaultValue=0.05, parameterName=Radius}
    Edge Falloff=ScalarParameter() {defaultValue=0.5, parameterName=Falloff}
    Square Corners?=StaticBool()
   ) {materialFunction=/Engine/Functions/Engine_MaterialFunctions01/Coordinates/DrawLine-2D}.Result
   B=MaterialFunctionCall(
    Coordinates=TextureCoordinate()
    P0=VectorParameter() {parameterName=A, defaultValue=<rgba 0.25 0.75 0 0>}.RGB
    P1=VectorParameter() {parameterName=B, defaultValue=<rgba 0.75 0.25 0 0>}.RGB
    Line Width=ScalarParameter() {defaultValue=0.05, parameterName=Radius}
    Edge Falloff=ScalarParameter() {defaultValue=0.5, parameterName=Falloff}
    Square Corners?=StaticBool()
   ) {materialFunction=&DrawLine-2D}.Result
  )
 ]
}
