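# Editor Utility Widget for the Landmass erosion brush that edits an ordered timeline of erosion keyframes and runs the simulation.
# Key widgets are sized by iteration share, edited through mode-specific detail views, and can be reordered or deleted.
# Simulate copies keys into the owning brush and starts or pauses it, while Tick advances a progress counter.
# Keys save to and load from preset data assets, and return to the brush on destruct.
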EditorUtilityWidgetBlueprint /Landmass/Landscape/ErosionBrush/BP/Widgets/EUW_Landscape_Erosion
graphs [
 Count Iterations {nodes [
   Inputs()
   > Set Max Iterations(Max Iterations:int=0) -> (Output_Get:int=0)
   > n1 = For Each Loop(Array:Array of EUW Erosion Key Object References=Get ErosionKeys() -> (ErosionKeys:Array of EUW Erosion Key Object References)) -> (Array Element:EUW Erosion Key Object Reference)
   Set Text(
    execute=@n1.Completed
    Text="To Text (Integer)"(Value=Get Max Iterations() -> (Max Iterations:int=0))
    self=Get Text_Iterations() -> (Text_Iterations:Text Object Reference)
   )
   > Outputs!(then:Exec)
   n2 = Break Landmass Erosion Custom Material!(Landmass_Erosion_CustomMaterial:"Landmass Erosion Custom Material Structure (by ref)"=Get Custom Material Pass!(self:EUW Erosion Key Object Reference=@n1) -> (Custom Material Pass:Landmass Erosion Custom Material Structure)) -> (
    Iterations_35_#1:int
    VectorParameters_36_#2:Map of Names to Linear Color Structures
   ) {is=title K2Node_BreakStruct}
   n3 = Break Landmass Particle Erosion Settings!(Landmass_Particle_Erosion_Settings:"Landmass Particle Erosion Settings Structure (by ref)"=Get Particle Erosion Settings!(self:EUW Erosion Key Object Reference=@n1) -> (Particle Erosion Settings:Landmass Particle Erosion Settings Structure)) -> (
    ErosionIterations_26_#3:int
    VelocityDamping_27_#4:double=~
    ErodeWeightmaps_31_#5:Array of Erode Weightmap Structures
    ResolutionDivisor_30_#6:int
   ) {is=title K2Node_BreakStruct}
   n4 = Break Landmass Water Levelling Settings!(Landmass_WaterLevelling_Settings:"Landmass Water Levelling Settings Structure (by ref)"=Get Water Levelling Settings!(self:EUW Erosion Key Object Reference=@n1) -> (Water Levelling Settings:Landmass Water Levelling Settings Structure)) -> (Iterations_28_#3:int, ErodeWeightmaps_30_#5:Array of Erode Weightmap Structures) {is=title K2Node_BreakStruct}
   n5 = Break Landmass Slope Smoothing Settings!(Landmass_SlopeSmoothing_Settings:"Landmass Slope Smoothing Settings Structure (by ref)"=Get Slope Smoothing Settings!(self:EUW Erosion Key Object Reference=@n1) -> (Slope Smoothing Settings:Landmass Slope Smoothing Settings Structure)) -> (Iterations_27_#3:int) {is=title K2Node_BreakStruct}
   n6 = Get Erosion Mode!(self:EUW Erosion Key Object Reference=@n1) -> (Erosion Mode:E_Landmass_CustomBrush_ErosionModes Enum=NewEnumerator0)
   Set Max Iterations(
    execute=@n1.LoopBody
    Max Iterations=int + int(
     A=Get Max Iterations() -> (Max Iterations=0)
     B=Select!(
      NewEnumerator0:int=@n3
      NewEnumerator3:int=@n5
      NewEnumerator4:int=@n4
      NewEnumerator1:int=1
      NewEnumerator2:int=@n2
      Index:E_Landmass_CustomBrush_ErosionModes Enum=@n6
     ) -> (ReturnValue:int)
    )
   ) -> (Output_Get=0)
  ]}
 EventGraph {nodes [
   "On Clicked (Button_AddErosionKey)"() {is=title K2Node_ComponentBoundEvent}
   > n1 = Create Widget!(
    execute:Exec
    WorldContextObject:Object Reference
    Class:User Widget Class Reference=&/EUW_ErosionKey.EUW_ErosionKey_C
    OwningPlayer:Player Controller Object Reference
   ) -> (then:Exec, ReturnValue:EUW Erosion Key Object Reference)
   > n2 = Cast To EUW_ErosionKey!(execute, Object=@n1) -> (then, CastFailed, AsEUW Erosion Key:EUW Erosion Key Object Reference) {is=title K2Node_DynamicCast}
   > Set ParentWidget!(execute, ParentWidget:EUW Landscape Erosion Object Reference=Self-Reference(), self:EUW Erosion Key Object Reference=@n2) -> (then, Output_Get:EUW Landscape Erosion Object Reference)
   > Add(TargetArray:Array of EUW Erosion Key Object References=Get ErosionKeys(), NewItem:"EUW Erosion Key Object Reference (by ref)"=@n2) {is=title K2Node_CallArrayFunction}
   > n3 = AddChildToHorizontalBox!(
    execute
    self:Horizontal Box Object Reference=Get HorizontalBox_ErosionKeys() -> (HorizontalBox_ErosionKeys:Horizontal Box Object Reference)
    Content:Widget Object Reference=@n2
   ) -> (then, ReturnValue:Horizontal Box Slot Object Reference)
   > SetSize!(
    execute
    self:Horizontal Box Slot Object Reference=@n3
    InSize:Slate Child Size Structure=Make Slate Child Size!(Value:float=1, SizeRule:ESlateSizeRule Enum=Fill) -> (SlateChildSize:Slate Child Size Structure) {is=title K2Node_MakeStruct}
   ) -> (then)
   > Count Iterations!(execute:Exec) -> (then:Exec) {is=title K2Node_MacroInstance}
   > Size Keys!(execute:Exec) -> (then:Exec) {is=title K2Node_MacroInstance}
   "On Clicked (Button_Simulate)"() {is=title K2Node_ComponentBoundEvent}
   > n4 = Branch(Condition=Get Simulating() -> (Simulating:bool))
   > Set Simulating(Simulating:bool=false) -> (Output_Get:bool)
   n5 = Get SimulateText() -> (SimulateText:Text Object Reference)
   > Set Text(Text="LOCTEXT(\"Simulate Erosion\")", self=@n5)
   n6 = Get Button_AddErosionKey() -> (Button_AddErosionKey:Editor Utility Button Object Reference)
   > SetIsEnabled(self=@n6, bInIsEnabled=true)
   > n7 = Set Paused(Paused:bool=true) -> (Output_Get:bool)
   Pause(
    execute=[@n7.then, @n9.then]
    self:Landmass Brush Erosion Object Reference=Get Owning_ErosionBrush() -> (Owning_ErosionBrush:Landmass Brush Erosion Object Reference)
   )
   Set Simulating(execute=@n4.else, Simulating=true)
   > Set Text(Text="LOCTEXT(\"Pause\")", self=@n5)
   > SetIsEnabled(self=@n6, bInIsEnabled=false)
   > n8 = Branch(Condition=Get Paused() -> (Paused:bool))
   > n9 = Set Paused(Paused=false)
   Clear(
    execute=@n8.else
    TargetArray:Array of Landmass Erosion Keyframe Structures=Get Erosion Keyframes!(self:Landmass Brush Erosion Object Reference=Get Owning_ErosionBrush()) -> (Erosion Keyframes:Array of Landmass Erosion Keyframe Structures)
   )
   > n10 = For Each Loop(Array=Get ErosionKeys())
   Simulate Erosion(execute=@n10.Completed, self:Landmass Brush Erosion Object Reference=Get Owning_ErosionBrush())
   > Set Current Iteration(Current Iteration:int=0) -> (Output_Get:int=0)
   Add(
    execute=@n10.LoopBody
    TargetArray:Array of Landmass Erosion Keyframe Structures=Get Erosion Keyframes(self=Get Owning_ErosionBrush())
    NewItem:"Landmass Erosion Keyframe Structure (by ref)"=Make Key Struct from Widget!(Key Widget:EUW Erosion Key Object Reference=@n10) -> (Landmass_Erosion_Keyframe:Landmass Erosion Keyframe Structure)
   )
   n11 = "On Property Changed (Details_WaterLevelling)"!() -> (OutputDelegate, then, PropertyName:Name=None) {is=title K2Node_ComponentBoundEvent}
   > Property Change!(execute:Exec, Selection:Name=@n11.PropertyName) {is=title K2Node_MacroInstance}
   n12 = "On Clicked (Button_SaveToDataAsset)"() {is=title K2Node_ComponentBoundEvent}
   Clear(
    execute=[@n12.then, @n35.then]
    TargetArray=Get ErosionKeyDataStructs() -> (ErosionKeyDataStructs:Array of Landmass Erosion Keyframe Structures)
   )
   > n13 = For Each Loop(Array=Get ErosionKeys())
   Set ErosionKeyData!(
    execute=@n13.Completed
    ErosionKeyData:Array of Landmass Erosion Keyframe Structures=Get ErosionKeyDataStructs()
    self:Erosion Brush Data Settings Preset Object Reference=Get SettingsPreset() -> (SettingsPreset:Erosion Brush Data Settings Preset Object Reference)
   ) -> (then, Output_Get:Array of Landmass Erosion Keyframe Structures)
   Add(execute=@n13.LoopBody, TargetArray=Get ErosionKeyDataStructs(), NewItem=Make Key Struct from Widget(Key Widget=@n13))
   "On Clicked (Button_LoadSettings)"() {is=title K2Node_ComponentBoundEvent}
   n14 = Get ErosionKeys()
   > n15 = For Each Loop(Array=@n14)
   Clear(execute=@n15.Completed, TargetArray:Array of EUW Erosion Key Object References=@n14)
   > n16 = For Each Loop(Array:Array of Landmass Erosion Keyframe Structures=Get ErosionKeyData!(self:Erosion Brush Data Settings Preset Object Reference=Get SettingsPreset()) -> (ErosionKeyData:Array of Landmass Erosion Keyframe Structures)) -> (Array Element:Landmass Erosion Keyframe Structure)
   Update Key Names!(execute:Exec=@n16.Completed) -> (then:Exec) {is=title K2Node_MacroInstance}
   > Count Iterations()
   > Size Keys()
   Make Key Widget from Key Struct!(execute=@n16.LoopBody, self, InKey:Landmass Erosion Keyframe Structure=@n16) -> (then)
   RemoveChild(execute=@n15.LoopBody, self=Get HorizontalBox_ErosionKeys(), Content=@n15)
   "On Clicked (Button_DataAssetCategory)"() {is=title K2Node_ComponentBoundEvent}
   K2Node_Composite_0 = IsCollapsed!(A:ESlateVisibility Enum=GetVisibility!(self:Widget Object Reference=Get VerticalBox_SettingsPreset() -> (VerticalBox_SettingsPreset:Vertical Box Object Reference)) -> (ReturnValue:ESlateVisibility Enum=Visible)) -> (Collapse or NonhitTest:ESlateVisibility Enum=Visible, Collapsed:bool=~) {is=title K2Node_Composite}
   > SetVisibility(self=Get VerticalBox_SettingsPreset(), InVisibility=@K2Node_Composite_0)
   > SetBrushFromMaterial(
    self=Get Image_SettingsPresetArrow() -> (Image_SettingsPresetArrow:Image Object Reference)
    Material=Select!(
     Option 0:Material Interface Object Reference=/Landmass/Landscape/ErosionBrush/Materials/Slate/CategoryArrow_Rotated
     Option 1:Material Interface Object Reference=/Landmass/Landscape/ErosionBrush/Materials/Slate/CategoryArrow
     Index:bool=@K2Node_Composite_0.Collapsed
    ) -> (ReturnValue:Material Interface Object Reference)
   )
   "On Clicked (Button_NewDataAsset)"() {is=title K2Node_ComponentBoundEvent}
   > SetVisibility(self=Get NewDataAsset_PathPrompt() -> (NewDataAsset_PathPrompt:Border Object Reference), InVisibility=SelfHitTestInvisible)
   "On Clicked (EditorUtilityButton_CreateAsset)"() {is=title K2Node_ComponentBoundEvent}
   > SetVisibility(self=Get NewDataAsset_PathPrompt(), InVisibility=Collapsed)
   > K2Node_Composite_2 = Create new DataAsset!(execute) -> (then:Exec) {is=title K2Node_Composite}
   > Apply Settings to DA()
   "On Clicked (Button_Reset)"() {is=title K2Node_ComponentBoundEvent}
   > Set Current Iteration(Current Iteration=0) -> (Output_Get=0)
   > Set CurTime(CurTime=0)
   > Set Paused(Paused=false)
   > Set Simulating(Simulating=false)
   > Set Text(Text="LOCTEXT(\"Simulate Erosion\")", self=Get SimulateText())
   > Reset(self:Landmass Brush Erosion Object Reference=Get Owning_ErosionBrush())
   n17 = "On Property Changed (Details_ParticleErosion)"!() -> (OutputDelegate, then, PropertyName:Name=None) {is=title K2Node_ComponentBoundEvent}
   > Property Change(Selection=@n17.PropertyName)
   n18 = "On Clicked (Button_Delete)"() {is=title K2Node_ComponentBoundEvent}
   n19 = Get SelectedKey() -> (SelectedKey:EUW Erosion Key Object Reference)
   Remove Item!(
    execute=[@n18.then, @n34.then]
    TargetArray:Array of EUW Erosion Key Object References=Get ErosionKeys()
    Item:"EUW Erosion Key Object Reference (by ref)"=@n19
   ) -> (then, ReturnValue:bool)
   > RemoveChild(self=Get HorizontalBox_ErosionKeys(), Content=@n19)
   > Count Iterations()
   > Size Keys()
   n20 = Get Details_ParticleErosion() -> (Details_ParticleErosion:Details View Object Reference)
   > SetVisibility(self=@n20, InVisibility=Collapsed)
   > SetObject(self=@n20)
   > Set SelectedKey(SelectedKey:EUW Erosion Key Object Reference) -> (Output_Get:EUW Erosion Key Object Reference)
   "On Clicked (Button_SwapBack)"() {is=title K2Node_ComponentBoundEvent}
   n21 = Find Item!(
    TargetArray:Array of EUW Erosion Key Object References=Get ErosionKeys()
    ItemToFind:"EUW Erosion Key Object Reference (by ref)"=Get SelectedKey()
   ) -> (ReturnValue)
   > Branch(Condition="integer >= integer"(A=@n21, B=1))
   > Swap Array Elements!(execute, TargetArray:Array of EUW Erosion Key Object References=Get ErosionKeys(), FirstIndex:int=int - int(A=@n21), SecondIndex:int=@n21) -> (then) {is=title K2Node_CallArrayFunction}
   > n22 = For Each Loop(Array:Array of EUW Erosion Key Object References=Get ErosionKeys()) -> (Array Element:EUW Erosion Key Object Reference)
   n23 = For Each Loop(Exec=@n22.Completed, Array=Get ErosionKeys())
   Size Keys(execute=@n23.Completed)
   AddChildToHorizontalBox(execute=@n23.LoopBody, self=Get HorizontalBox_ErosionKeys(), Content=@n23)
   RemoveChild(execute=@n22.LoopBody, self=Get HorizontalBox_ErosionKeys(), Content=@n22)
   "On Clicked (Button_SwapFwd)"() {is=title K2Node_ComponentBoundEvent}
   n24 = Get ErosionKeys()
   n25 = Find Item(TargetArray=@n24, ItemToFind=Get SelectedKey())
   > Branch(Condition="integer < integer"(A=@n25, B=Last Index!(TargetArray:Array of EUW Erosion Key Object References=@n24) -> (ReturnValue=0)))
   > Swap Array Elements(TargetArray=Get ErosionKeys(), FirstIndex=@n25, SecondIndex=int + int(A=@n25))
   > n26 = For Each Loop(Array=Get ErosionKeys())
   n27 = For Each Loop(Exec=@n26.Completed, Array=Get ErosionKeys())
   Size Keys(execute=@n27.Completed)
   AddChildToHorizontalBox(execute=@n27.LoopBody, self=Get HorizontalBox_ErosionKeys(), Content=@n27)
   RemoveChild(execute=@n26.LoopBody, self=Get HorizontalBox_ErosionKeys(), Content=@n26)
   n28 = "On Property Changed (Details_NoiseInjection)"!() -> (OutputDelegate, then, PropertyName:Name=None) {is=title K2Node_ComponentBoundEvent}
   > Property Change(Selection=@n28.PropertyName)
   n29 = "On Property Changed (Details_CustomMaterial)"!() -> (OutputDelegate, then, PropertyName:Name=None) {is=title K2Node_ComponentBoundEvent}
   > Property Change(Selection=@n29.PropertyName)
   n30 = "On Property Changed (Details_SlopeSmoothing)"!() -> (OutputDelegate, then, PropertyName:Name=None) {is=title K2Node_ComponentBoundEvent}
   > Property Change(Selection=@n30.PropertyName)
   n31 = Clicked Erosion Key!() -> (OutputDelegate, then, Key:EUW Erosion Key Object Reference) {is=title K2Node_CustomEvent}
   > n32 = For Each Loop(Array=Get ErosionKeys())
   n33 = Branch(execute=@n32.Completed, Condition="Equal (Object)"(A=Get SelectedKey(), B=@n31.Key))
   > Set SelectedKey()
   > SetVisibility(
    self=[Get Details_ParticleErosion(), Get HorizontalBox_MoveDelete() -> (HorizontalBox_MoveDelete:Horizontal Box Object Reference)]
    InVisibility=Collapsed
   )
   SetStyle!(
    execute=@n33.else
    self:Button Object Reference=Get Button_0!(self:EUW Erosion Key Object Reference=@n31.Key) -> (Button_0:Button Object Reference)
    InStyle:"Button Style Structure (by ref)"=Get Selected_ButtonStyle!(self:EUW Erosion Key Object Reference=@n31.Key) -> (Selected_ButtonStyle:Button Style Structure)
   ) -> (then)
   > Set SelectedKey(SelectedKey=@n31.Key)
   > SetVisibility(self=Get HorizontalBox_MoveDelete(), InVisibility=SelfHitTestInvisible)
   > Update Detail Views!(execute:Exec) -> (then:Exec) {is=title K2Node_MacroInstance}
   SetStyle(
    execute=@n32.LoopBody
    self=Get Button_0(self=@n32)
    InStyle=Get Default_ButtonStyle!(self:EUW Erosion Key Object Reference=@n32) -> (Default_ButtonStyle:Button Style Structure)
   )
   n34 = Delete a Key() {is=title K2Node_CustomEvent}
   n35 = Apply Settings to DA() {is=title K2Node_CustomEvent}
   Event Construct()
   > Delay!(execute, WorldContextObject:Object Reference, Duration=0.2) -> (then)
   > SetObject(self=Get SinglePropertyView_80() -> (SinglePropertyView_80:Single Property View Object Reference), NewObject=Self-Reference())
   > n36 = For Each Loop(Array:Array of Landmass Erosion Keyframe Structures=Get Erosion Keyframes(self=Get Owning_ErosionBrush())) -> (Array Element:Landmass Erosion Keyframe Structure)
   Update Key Names(execute=@n36.Completed)
   > Count Iterations()
   > Size Keys()
   Make Key Widget from Key Struct(execute=@n36.LoopBody, InKey=@n36)
   Event Tick()
   > Branch(Condition=Get Simulating())
   n37 = "To Float (Integer)"(InInt=Get Max Iterations() -> (Max Iterations=0))
   n38 = "To Float (Integer)"(InInt=Get Current Iteration() -> (Current Iteration:int=0))
   > Set CurTime(CurTime="Clamp (Float)"(Value=float / float(A=@n38, B=@n37) {is=title K2Node_CallFunction}, Min=0, Max=1))
   n39 = Get Owning_ErosionBrush()
   > Set Current Iteration(Current Iteration=int + int(
     A=Get Current Iteration() -> (Current Iteration=0)
     B=Get Iterations Per Tick!(self:Landmass Brush Erosion Object Reference=@n39) -> (Iterations Per Tick:int=0)
    )) -> (Output_Get=0)
   > Branch(Condition="integer >= integer"(A=Get Current Iteration() -> (Current Iteration=0), B=Get Max Iterations() -> (Max Iterations=0)))
   > Set Simulating(Simulating=false)
   > PrintString(InString="Erosion Widget: Simulation Complete!")
   > Set Text(Text="LOCTEXT(\"Simulate Erosion\")", self=Get SimulateText())
   > SetIsEnabled(self=Get Button_AddErosionKey(), bInIsEnabled=true)
   Event Destruct()
   > Clear(TargetArray:Array of Landmass Erosion Keyframe Structures=Get Erosion Keyframes(self=Get Owning_ErosionBrush()))
   > n40 = For Each Loop(Array:Array of EUW Erosion Key Object References=Get ErosionKeys()) -> (Array Element:EUW Erosion Key Object Reference)
   > Add(TargetArray=Get Erosion Keyframes(self=Get Owning_ErosionBrush()), NewItem=Make Key Struct from Widget(Key Widget=@n40))
   Event PreConstruct()
  ]}
 IsCollapsed {
  path=EventGraph.K2Node_Composite_0.IsCollapsed
  nodes [
   n1 = "Equal (Enum)"(A:ESlateVisibility Enum=Inputs!() -> (A:ESlateVisibility Enum=Visible), B:ESlateVisibility Enum=Collapsed)
   Outputs!(
    Collapse or NonhitTest:ESlateVisibility Enum=Select(Option 0:ESlateVisibility Enum=Collapsed, Option 1:ESlateVisibility Enum=SelfHitTestInvisible, Index=@n1) -> (ReturnValue:ESlateVisibility Enum)
    Collapsed:bool=@n1
   )
  ]
 }
 Create new DataAsset {
  path=EventGraph.K2Node_Composite_2.Create new DataAsset
  nodes [
   Inputs!() -> (execute:Exec)
   > n1 = GetAssetTools()
   n2 = Get NewDataAssetPath() -> (NewDataAssetPath:Editor Utility Editable Text Box Object Reference)
   n3 = Split(SourceString="To String (Text)"(InText=Get Text(self=@n2)), InStr=/, SearchDir=FromEnd)
   > n4 = CreateAsset(
    self=@n1
    AssetName=@n3.RightS
    PackagePath=@n3
    AssetClass=/Landmass/Landscape/ErosionBrush/BP/DataAsset/ErosionBrush_DataSettings_Preset.ErosionBrush_DataSettings_Preset_C
   )
   > n5 = Cast To ErosionBrush_DataSettings_Preset!(execute, Object=@n4) -> (then, CastFailed, AsErosion Brush Data Settings Preset:Erosion Brush Data Settings Preset Object Reference) {is=title K2Node_DynamicCast}
   > Set SettingsPreset(SettingsPreset:Erosion Brush Data Settings Preset Object Reference=@n5) -> (Output_Get:Erosion Brush Data Settings Preset Object Reference)
   > Outputs!(then:Exec)
  ]
 }
 Make Key Struct from Widget {nodes [
   n1 = Make Key Struct from Widget!() -> (then, Key Widget:EUW Erosion Key Object Reference) {is=title K2Node_FunctionEntry}
   > Return Node!(
    execute
    Landmass_Erosion_Keyframe:Landmass Erosion Keyframe Structure=Make Landmass Erosion Keyframe!(
     KeyName_6_#7:Text=Get Key Name!(self:EUW Erosion Key Object Reference=@n1) -> (Key Name:Text)
     ErosionMode_23_#8:E_Landmass_CustomBrush_ErosionModes Enum=Get Erosion Mode(self=@n1) -> (Erosion Mode=NewEnumerator0)
     ParticleErosion_7_#9:Landmass Particle Erosion Settings Structure=Get Particle Erosion Settings(self=@n1)
     SlopeSmoothing_17_#10:Landmass Slope Smoothing Settings Structure=Get Slope Smoothing Settings(self=@n1)
     WaterLevelling_18_#11:Landmass Water Levelling Settings Structure=Get Water Levelling Settings(self=@n1)
     NoiseInjection_19_#12:Landmass Erosion Noise Injection Structure=Get Noise Injection Settings!(self:EUW Erosion Key Object Reference=@n1) -> (Noise Injection Settings:Landmass Erosion Noise Injection Structure)
     CustomMaterial_20_#13:Landmass Erosion Custom Material Structure=Get Custom Material Pass(self=@n1)
    ) -> (Landmass_Erosion_Keyframe:Landmass Erosion Keyframe Structure) {is=title K2Node_MakeStruct}
   )
   Make Landmass Erosion Keyframe(
    KeyName_6_#7="LOCTEXT(\"Erosion Keyframe\")"
    ErosionMode_23_#8=NewEnumerator0
    ParticleErosion_7_#9=(ErosionIterations_26_#3=1000,ResetParticlesEveryNIterations_2_#14=0,StartingWaterAmount_5_#15=100,DeltaT_7_#16=0.0167,Solubility_9_#17=0,Evaporation_11_#18=0,VelocityDamping_27_#4=0.2,VelocityAdvection_16_#19=0,ErodeWeightmaps_31_#5=,ResolutionDivisor_30_#6=1)
    SlopeSmoothing_17_#10=(Iterations_27_#3=100,SmoothSlopesGreaterThan_18_#4=0,SmoothSlopesLessThan_20_#20=0,Opacity_30_#21=1)
    WaterLevelling_18_#11=(Iterations_28_#3=100,ResetWaterEveryNIterations_29_#14=0,StartingWaterAmount_5_#15=10,RainPerIteration_27_#16=0.0167,Solubility_9_#17=0.1,Evaporation_11_#18=0.1,ErodeWeightmaps_30_#5=)
    NoiseInjection_19_#12=(NoiseHeightMeters_20_#22=2,NoiseScaleMeters_21_#23=100,NoiseOctaves_4_#24=6,NoisePersistence_11_#25=0.5,NoiseSeed_14_#26=0,Turbulence_16_#27=False,NoiseFloor0-1_22_#28=0.25,SlopeMask_25_#29=(EnableSlopeMask_10_#30=False,MinSlope_11_#31=0.25,MaxSlope_12_#32=1,SlopeFalloff_13_#33=0.1))
    CustomMaterial_20_#13=(Iterations_35_#1=100,Material_19_#34=None,ScalarParameters_24_#35=(),VectorParameters_36_#2=(),TextureParameters_33_#36=())
   )
  ]}
 Make Key Widget from Key Struct {nodes [
   n1 = Make Key Widget from Key Struct!() -> (then, InKey:Landmass Erosion Keyframe Structure) {is=title K2Node_FunctionEntry}
   > n2 = Construct Object from Class(Class=&EUW_ErosionKey_C) -> (ReturnValue:EUW Erosion Key Object Reference)
   > n3 = Set NewWidget(NewWidget:EUW Erosion Key Object Reference=@n2) -> (Output_Get:EUW Erosion Key Object Reference)
   > Set ParentWidget(ParentWidget=Self-Reference(), self=@n3)
   > Add(
    TargetArray:Array of EUW Erosion Key Object References=Get ErosionKeys()
    NewItem:"EUW Erosion Key Object Reference (by ref)"=Get NewWidget() -> (NewWidget:EUW Erosion Key Object Reference)
   )
   n4 = Break Landmass Erosion Keyframe!(Landmass_Erosion_Keyframe:"Landmass Erosion Keyframe Structure (by ref)"=@n1) -> (
    KeyName_6_#7:Text
    ErosionMode_23_#8:E_Landmass_CustomBrush_ErosionModes Enum
    ParticleErosion_7_#9:Landmass Particle Erosion Settings Structure
    SlopeSmoothing_17_#10:Landmass Slope Smoothing Settings Structure
    WaterLevelling_18_#11:Landmass Water Levelling Settings Structure
    NoiseInjection_19_#12:Landmass Erosion Noise Injection Structure
    CustomMaterial_20_#13:Landmass Erosion Custom Material Structure
   ) {is=title K2Node_BreakStruct}
   > Set Key Name!(execute, Key Name:Text=@n4, self:EUW Erosion Key Object Reference=Get NewWidget()) -> (then, Output_Get:Text)
   > Set Erosion Mode!(
    execute
    Erosion Mode:E_Landmass_CustomBrush_ErosionModes Enum=@n4.ErosionMode_23_#8
    self:EUW Erosion Key Object Reference=Get NewWidget()
   ) -> (then, Output_Get:E_Landmass_CustomBrush_ErosionModes Enum=NewEnumerator0)
   > Set Particle Erosion Settings!(
    execute
    Particle Erosion Settings:Landmass Particle Erosion Settings Structure=@n4.ParticleErosion_7_#9
    self:EUW Erosion Key Object Reference=Get NewWidget()
   ) -> (then, Output_Get:Landmass Particle Erosion Settings Structure)
   > Set Slope Smoothing Settings!(
    execute
    Slope Smoothing Settings:Landmass Slope Smoothing Settings Structure=@n4.SlopeSmoothing_17_#10
    self:EUW Erosion Key Object Reference=Get NewWidget()
   ) -> (then, Output_Get:Landmass Slope Smoothing Settings Structure)
   > Set Water Levelling Settings!(
    execute
    Water Levelling Settings:Landmass Water Levelling Settings Structure=@n4.WaterLevelling_18_#11
    self:EUW Erosion Key Object Reference=Get NewWidget()
   ) -> (then, Output_Get:Landmass Water Levelling Settings Structure)
   > Set Noise Injection Settings!(
    execute
    Noise Injection Settings:Landmass Erosion Noise Injection Structure=@n4.NoiseInjection_19_#12
    self:EUW Erosion Key Object Reference=Get NewWidget()
   ) -> (then, Output_Get:Landmass Erosion Noise Injection Structure)
   > Set Custom Material Pass!(
    execute
    Custom Material Pass:Landmass Erosion Custom Material Structure=@n4.CustomMaterial_20_#13
    self:EUW Erosion Key Object Reference=Get NewWidget()
   ) -> (then, Output_Get:Landmass Erosion Custom Material Structure)
   > n5 = AddChildToHorizontalBox(self=Get HorizontalBox_ErosionKeys(), Content=Get NewWidget())
   > SetSize(self=@n5, InSize=Make Slate Child Size(Value=1, SizeRule=Fill))
  ]}
 OnKeyDown {nodes [
   OnKeyDown!() -> (then, MyGeometry:Geometry Structure, InKeyEvent:Key Event Structure) {is=title K2Node_FunctionEntry}
   > n1 = Sequence()
   n2 = Get InKeyEvent() -> (InKeyEvent:Key Event Structure)
   > Branch(Condition="Equal (Key)"(A:Key Structure=GetKey!(Input:"Key Event Structure (by ref)"=@n2) -> (ReturnValue:Key Structure), B:Key Structure=Delete) -> (ReturnValue:bool=~) {is=title K2Node_PromotableOperator})
   > Delete a Key() {is=title K2Node_CallFunction}
   Return Node!(execute=@n1.then_1, ReturnValue:Event Reply Structure=Make Event Reply!() -> (EventReply:Event Reply Structure) {is=title K2Node_MakeStruct})
  ]}
 Property Change {nodes [
   n1 = Inputs!() -> (execute, Selection:Name)
   > n2 = Switch on Name!(execute:Exec, Selection:Name=@n1) -> (Default:Exec, Key Name:Exec, Erosion Mode:Exec) {is=title K2Node_SwitchName, pinNames [Key Name, Erosion Mode]}
   > Count Iterations()
   > Size Keys()
   Update Detail Views(execute=@n2.Erosion Mode)
   > Count Iterations()
   > Size Keys()
   Update Key Names(execute=@n2.Key Name)
   Outputs!()
  ]}
 Size Keys {nodes [
   Inputs!() -> (execute)
   > n1 = For Each Loop(Array=Get ErosionKeys())
   Outputs!(then:Exec=@n1.Completed)
   > SetVisibility(
    self=Get VerticalBox_311!(self:EUW Erosion Key Object Reference=@n1) -> (VerticalBox_311:Vertical Box Object Reference)
    InVisibility=Select!(
     NewEnumerator0:ESlateVisibility Enum=HitTestInvisible
     NewEnumerator3:ESlateVisibility Enum=Visible
     NewEnumerator4:ESlateVisibility Enum=Visible
     NewEnumerator1:ESlateVisibility Enum=Collapsed
     NewEnumerator2:ESlateVisibility Enum=HitTestInvisible
     Index:E_Landmass_CustomBrush_ErosionModes Enum=Get Erosion Mode(self=@n1) -> (Erosion Mode=NewEnumerator0)
    ) -> (ReturnValue)
   )
   > n2 = Cast To HorizontalBoxSlot!(execute, Object=Get Slot!(self:Widget Object Reference=@n1) -> (Slot:Panel Slot Object Reference)) -> (then, CastFailed, AsHorizontal Box Slot:Horizontal Box Slot Object Reference) {is=title K2Node_DynamicCast}
   n3 = Break Landmass Erosion Custom Material(Landmass_Erosion_CustomMaterial=Get Custom Material Pass(self=@n1))
   n4 = Break Landmass Particle Erosion Settings(Landmass_Particle_Erosion_Settings=Get Particle Erosion Settings(self=@n1)) -> (VelocityDamping_27_#4=~)
   n5 = Break Landmass Water Levelling Settings(Landmass_WaterLevelling_Settings=Get Water Levelling Settings(self=@n1))
   n6 = Break Landmass Slope Smoothing Settings(Landmass_SlopeSmoothing_Settings=Get Slope Smoothing Settings(self=@n1))
   n7 = Get Erosion Mode(self=@n1) -> (Erosion Mode=NewEnumerator0)
   n8 = float / float(
    A="To Float (Integer)"(InInt=Select(NewEnumerator0:int=@n4, NewEnumerator3:int=@n6, NewEnumerator4:int=@n5, NewEnumerator1:int=1, NewEnumerator2:int=@n3, Index=@n7) -> (ReturnValue:int))
    B="To Float (Integer)"(InInt=Get Max Iterations() -> (Max Iterations=0))
   ) {is=title K2Node_PromotableOperator}
   > SetSize(self=@n2, InSize=Make Slate Child Size(Value="Max (Float)"(A=@n8, B=0.005), SizeRule=Fill))
  ]}
 Update Detail Views {nodes [
   Inputs()
   > n1 = Sequence!(execute) -> (then_0, then_1, then_2:Exec, then_3:Exec, then_4:Exec)
   n2 = Get SelectedKey()
   > n3 = Branch(Condition="Equal (Enum)"(
     A:E_Landmass_CustomBrush_ErosionModes Enum=Get Erosion Mode(self=@n2) -> (Erosion Mode=NewEnumerator0)
     B:E_Landmass_CustomBrush_ErosionModes Enum=NewEnumerator0
    ))
   n4 = Get Details_ParticleErosion()
   > SetVisibility(self=@n4, InVisibility=Visible)
   > SetObject(self=@n4, NewObject=Get SelectedKey())
   SetVisibility(execute=@n3.else, self=@n4, InVisibility=Collapsed)
   n5 = Get SelectedKey()
   n6 = Branch(execute=@n1.then_3, Condition="Equal (Enum)"(A=Get Erosion Mode(self=@n5) -> (Erosion Mode=NewEnumerator0), B=NewEnumerator3))
   n7 = Get Details_SlopeSmoothing() -> (Details_SlopeSmoothing:Details View Object Reference)
   > SetVisibility(self=@n7, InVisibility=Visible)
   > SetObject(self=@n7, NewObject=Get SelectedKey())
   SetVisibility(execute=@n6.else, self=@n7, InVisibility=Collapsed)
   n8 = Get SelectedKey()
   n9 = Branch(execute=@n1.then_4, Condition="Equal (Enum)"(A=Get Erosion Mode(self=@n8) -> (Erosion Mode=NewEnumerator0), B=NewEnumerator4))
   n10 = Get Details_WaterLevelling() -> (Details_WaterLevelling:Details View Object Reference)
   > SetVisibility(self=@n10, InVisibility=Visible)
   > n11 = SetObject(self=@n10, NewObject=Get SelectedKey())
   n12 = Get SelectedKey()
   SetVisibility(
    execute=[@n11.then, @n13.then]
    self=Get VerticalBox_311(self=@n12)
    InVisibility=Select(
     NewEnumerator0:ESlateVisibility Enum=HitTestInvisible
     NewEnumerator3:ESlateVisibility Enum=Visible
     NewEnumerator4:ESlateVisibility Enum=Visible
     NewEnumerator1:ESlateVisibility Enum=Collapsed
     NewEnumerator2:ESlateVisibility Enum=HitTestInvisible
     Index=Get Erosion Mode(self=@n12) -> (Erosion Mode=NewEnumerator0)
    ) -> (ReturnValue:ESlateVisibility Enum)
   )
   > Outputs()
   n13 = SetVisibility(execute=@n9.else, self=@n10, InVisibility=Collapsed)
   n14 = Get SelectedKey()
   n15 = Branch(execute=@n1.then_1, Condition="Equal (Enum)"(A=Get Erosion Mode(self=@n14) -> (Erosion Mode=NewEnumerator0), B=NewEnumerator1))
   n16 = Get Details_NoiseInjection() -> (Details_NoiseInjection:Details View Object Reference)
   > SetVisibility(self=@n16, InVisibility=Visible)
   > SetObject(self=@n16, NewObject=Get SelectedKey())
   SetVisibility(execute=@n15.else, self=@n16, InVisibility=Collapsed)
   n17 = Get SelectedKey()
   n18 = Branch(execute=@n1.then_2, Condition="Equal (Enum)"(A=Get Erosion Mode(self=@n17) -> (Erosion Mode=NewEnumerator0), B=NewEnumerator2))
   n19 = Get Details_CustomMaterial() -> (Details_CustomMaterial:Details View Object Reference)
   > SetVisibility(self=@n19, InVisibility=Visible)
   > SetObject(self=@n19, NewObject=Get SelectedKey())
   SetVisibility(execute=@n18.else, self=@n19, InVisibility=Collapsed)
  ]}
 Update Key Names {nodes [
   Inputs()
   > n1 = For Each Loop(Array=Get ErosionKeys())
   Outputs(then=@n1.Completed)
   > Set Text(
    Text=Get Key Name(self=@n1)
    self=Get TextBlock_KeyName!(self:EUW Erosion Key Object Reference=@n1) -> (TextBlock_KeyName:Text Object Reference)
   )
  ]}
]
variables [
 Current Iteration {type=integer}
 CurTime {type=number}
 ErosionKeyDataStructs {type="array<Landmass_Erosion_Keyframe>", default []}
 ErosionKeys {type="array<EUW_ErosionKey_C>", default []}
 Max Iterations {type=integer, default=100}
 Owning_ErosionBrush {type=LandmassBrush_Erosion_C, default=null}
 Paused {type=boolean}
 SelectedKey {type=EUW_ErosionKey_C, default=null}
 SettingsPreset {type=ErosionBrush_DataSettings_Preset_C, default=/Landmass/Landscape/ErosionBrush/BP/Presets/ErosionBrush_Data_Preset_Heavy_01}
 Simulating {type=boolean}
]
widgets {
 info {rootWidgetClass=&VerticalBox, widgetCount=51}
 widgets [
  VerticalBox_33 {widgetClassPath=&VerticalBox, widget=VerticalBox_33}
  HorizontalBox_34 {widgetClassPath=&HorizontalBox, parent=VerticalBox_33, widget=HorizontalBox_34, slot {class=VerticalBoxSlot}}
  Button_AddErosionKey {widgetClassPath=&EditorUtilityButton, parent=HorizontalBox_34, widget=Button_AddErosionKey, slot {class=HorizontalBoxSlot}}
  TextBlock_37 {
   widgetClassPath=&TextBlock
   parent=Button_AddErosionKey
   widget=TextBlock_37
   properties {
    text=Add Erosion Key
    font {
     bForceMonospaced=false
     bMaterialIsStencil=false
     fontMaterial=null
     fontObject=&Roboto
     letterSpacing=0
     monospacedWidth=1
     size=12
     skewAmount=0
     typefaceFontName=Bold
     outlineSettings {
      bApplyOutlineToDropShadows=false
      bMiteredCorners=false
      bSeparateFillAlpha=false
      outlineMaterial=null
      outlineSize=0
      outlineColor=<rgba 0 0 0 1>
     }
    }
   }
   slot {class=ButtonSlot}
  }
  SizeBox_147 {
   widgetClassPath=&SizeBox
   parent=HorizontalBox_34
   widget=SizeBox_147
   properties {bOverride_WidthOverride=true, widthOverride=150}
   slot {class=HorizontalBoxSlot}
  }
  Button_Simulate {widgetClassPath=&EditorUtilityButton, parent=SizeBox_147, widget=Button_Simulate, slot {class=SizeBoxSlot}}
  SimulateText {
   widgetClassPath=&TextBlock
   bIsVariable=true
   parent=Button_Simulate
   widget=SimulateText
   properties {
    text=Simulate Erosion
    font {
     bForceMonospaced=false
     bMaterialIsStencil=false
     fontMaterial=null
     fontObject=&Roboto
     letterSpacing=0
     monospacedWidth=1
     size=12
     skewAmount=0
     typefaceFontName=Bold
     outlineSettings {
      bApplyOutlineToDropShadows=false
      bMiteredCorners=false
      bSeparateFillAlpha=false
      outlineMaterial=null
      outlineSize=0
      outlineColor=<rgba 0 0 0 1>
     }
    }
   }
   slot {class=ButtonSlot}
  }
  Button_Reset {widgetClassPath=&EditorUtilityButton, parent=HorizontalBox_34, widget=Button_Reset, slot {class=HorizontalBoxSlot}}
  TextBlock_1 {
   widgetClassPath=&TextBlock
   parent=Button_Reset
   widget=TextBlock_1
   properties {
    text=Reset
    font {
     bForceMonospaced=false
     bMaterialIsStencil=false
     fontMaterial=null
     fontObject=&Roboto
     letterSpacing=0
     monospacedWidth=1
     size=12
     skewAmount=0
     typefaceFontName=Bold
     outlineSettings {
      bApplyOutlineToDropShadows=false
      bMiteredCorners=false
      bSeparateFillAlpha=false
      outlineMaterial=null
      outlineSize=0
      outlineColor=<rgba 0 0 0 1>
     }
    }
   }
   slot {class=ButtonSlot}
  }
  TextBlock_102 {
   widgetClassPath=&TextBlock
   parent=HorizontalBox_34
   widget=TextBlock_102
   properties {
    text="Total Iterations: "
    font {
     bForceMonospaced=false
     bMaterialIsStencil=false
     fontMaterial=null
     fontObject=&Roboto
     letterSpacing=0
     monospacedWidth=1
     size=12
     skewAmount=0
     typefaceFontName=Bold
     outlineSettings {
      bApplyOutlineToDropShadows=false
      bMiteredCorners=false
      bSeparateFillAlpha=false
      outlineMaterial=null
      outlineSize=0
      outlineColor=<rgba 0 0 0 1>
     }
    }
   }
   slot {class=HorizontalBoxSlot, properties {verticalAlignment=VAlign_Center}}
  }
  Text_Iterations {
   widgetClassPath=&TextBlock
   bIsVariable=true
   parent=HorizontalBox_34
   widget=Text_Iterations
   properties {
    text=0
    font {
     bForceMonospaced=false
     bMaterialIsStencil=false
     fontMaterial=null
     fontObject=&Roboto
     letterSpacing=0
     monospacedWidth=1
     size=12
     skewAmount=0
     typefaceFontName=Bold
     outlineSettings {
      bApplyOutlineToDropShadows=false
      bMiteredCorners=false
      bSeparateFillAlpha=false
      outlineMaterial=null
      outlineSize=0
      outlineColor=<rgba 0 0 0 1>
     }
    }
   }
   slot {class=HorizontalBoxSlot, properties {verticalAlignment=VAlign_Center}}
  }
  VerticalBox_440 {
   widgetClassPath=&VerticalBox
   parent=VerticalBox_33
   widget=VerticalBox_440
   slot {class=VerticalBoxSlot, properties {size {sizeRule=Fill, value=1}}}
  }
  HorizontalBox_ErosionKeys {
   widgetClassPath=&HorizontalBox
   bIsVariable=true
   parent=VerticalBox_440
   widget=HorizontalBox_ErosionKeys
   slot {class=VerticalBoxSlot, properties {padding=<ltrb 0 5 0 5>}}
  }
  SizeBox_32 {
   widgetClassPath=&SizeBox
   parent=HorizontalBox_ErosionKeys
   widget=SizeBox_32
   properties {bOverride_HeightOverride=true, heightOverride=50}
   slot {class=HorizontalBoxSlot}
  }
  Slider_ErosionTimeline {
   widgetClassPath=/Script/UMG.Slider
   bIsVariable=true
   parent=VerticalBox_440
   widget=Slider_ErosionTimeline
   properties {
    isFocusable=false
    widgetStyle {
     barThickness=4
     disabledBarImage {
      bIsDynamicallyLoaded=false
      drawAs=RoundedBox
      imageType=NoImage
      mirroring=NoMirror
      resourceName=None
      resourceObject=null
      tiling=NoTile
      imageSize=<0 0>
      margin=<ltrb 0 0 0 0>
      outlineSettings {
       bUseBrushTransparency=false
       roundingType=FixedRadius
       width=1
       color {colorUseRule=UseColor_ColorTable, specifiedColor=<rgba 0 0 0 0>}
       cornerRadii=<2 2 2 2>
      }
      tintColor {colorUseRule=UseColor_Specified, specifiedColor=<rgba 0.0048 0.0048 0.0048 1>}
      uVRegion {bIsValid=false, max=<0 0>, min=<0 0>}
     }
     disabledThumbImage {
      bIsDynamicallyLoaded=false
      drawAs=RoundedBox
      imageType=NoImage
      mirroring=NoMirror
      resourceName=None
      resourceObject=null
      tiling=NoTile
      imageSize=<8 8>
      margin=<ltrb 0 0 0 0>
      outlineSettings {
       bUseBrushTransparency=false
       roundingType=HalfHeightRadius
       width=0
       color {colorUseRule=UseColor_Specified, specifiedColor=<rgba 0 0 0 0>}
       cornerRadii=<0 0 0 0>
      }
      tintColor {colorUseRule=UseColor_Specified, specifiedColor=<rgba 0.5271 0.5271 0.5271 1>}
      uVRegion {bIsValid=false, max=<0 0>, min=<0 0>}
     }
     hoveredBarImage {
      bIsDynamicallyLoaded=false
      drawAs=RoundedBox
      imageType=NoImage
      mirroring=NoMirror
      resourceName=None
      resourceObject=null
      tiling=NoTile
      imageSize=<0 0>
      margin=<ltrb 0 0 0 0>
      outlineSettings {
       bUseBrushTransparency=false
       roundingType=FixedRadius
       width=1
       color {colorUseRule=UseColor_ColorTable, specifiedColor=<rgba 0 0 0 0>}
       cornerRadii=<2 2 2 2>
      }
      tintColor {colorUseRule=UseColor_Specified, specifiedColor=<rgba 0.0048 0.0048 0.0048 1>}
      uVRegion {bIsValid=false, max=<0 0>, min=<0 0>}
     }
     hoveredThumbImage {
      bIsDynamicallyLoaded=false
      drawAs=RoundedBox
      imageType=NoImage
      mirroring=NoMirror
      resourceName=None
      resourceObject=null
      tiling=NoTile
      imageSize=<8 8>
      margin=<ltrb 0 0 0 0>
      outlineSettings {
       bUseBrushTransparency=false
       roundingType=HalfHeightRadius
       width=0
       color {colorUseRule=UseColor_Specified, specifiedColor=<rgba 0 0 0 0>}
       cornerRadii=<0 0 0 0>
      }
      tintColor {colorUseRule=UseColor_Specified, specifiedColor=<rgba 1 1 1 1>}
      uVRegion {bIsValid=false, max=<0 0>, min=<0 0>}
     }
     normalBarImage {
      bIsDynamicallyLoaded=false
      drawAs=RoundedBox
      imageType=NoImage
      mirroring=NoMirror
      resourceName=None
      resourceObject=null
      tiling=NoTile
      imageSize=<0 0>
      margin=<ltrb 0 0 0 0>
      outlineSettings {
       bUseBrushTransparency=false
       roundingType=FixedRadius
       width=1
       color {colorUseRule=UseColor_ColorTable, specifiedColor=<rgba 0 0 0 0>}
       cornerRadii=<2 2 2 2>
      }
      tintColor {colorUseRule=UseColor_Specified, specifiedColor=<rgba 0.0048 0.0048 0.0048 1>}
      uVRegion {bIsValid=false, max=<0 0>, min=<0 0>}
     }
     normalThumbImage {
      bIsDynamicallyLoaded=false
      drawAs=RoundedBox
      imageType=NoImage
      mirroring=NoMirror
      resourceName=None
      resourceObject=null
      tiling=NoTile
      imageSize=<8 8>
      margin=<ltrb 0 0 0 0>
      outlineSettings {
       bUseBrushTransparency=false
       roundingType=HalfHeightRadius
       width=0
       color {colorUseRule=UseColor_Specified, specifiedColor=<rgba 0 0 0 0>}
       cornerRadii=<0 0 0 0>
      }
      tintColor {colorUseRule=UseColor_Specified, specifiedColor=<rgba 0.5271 0.5271 0.5271 1>}
      uVRegion {bIsValid=false, max=<0 0>, min=<0 0>}
     }
    }
   }
   slot {class=VerticalBoxSlot, properties {verticalAlignment=VAlign_Bottom}}
  }
  HorizontalBox_MoveDelete {
   widgetClassPath=&HorizontalBox
   bIsVariable=true
   parent=VerticalBox_440
   widget=HorizontalBox_MoveDelete
   properties {visibility=Collapsed}
   slot {class=VerticalBoxSlot, properties {horizontalAlignment=HAlign_Left, verticalAlignment=VAlign_Top, padding=<ltrb 0 5 0 10>}}
  }
  Button_SwapBack {
   widgetClassPath=&EditorUtilityButton
   parent=HorizontalBox_MoveDelete
   widget=Button_SwapBack
   properties {toolTipText=Move Key to Left}
   slot {class=HorizontalBoxSlot}
  }
  Image_59 {
   widgetClassPath=&Image
   parent=Button_SwapBack
   widget=Image_59
   properties {brush {
     bIsDynamicallyLoaded=false
     drawAs=Image
     imageType=NoImage
     mirroring=NoMirror
     resourceName=None
     resourceObject=/Landmass/Landscape/OldPrototype_BP/Mats/Slate/M_Button_Arrow_Back
     tiling=NoTile
     imageSize=<32 32>
     margin=<ltrb 0 0 0 0>
     outlineSettings {
      bUseBrushTransparency=false
      roundingType=HalfHeightRadius
      width=0
      color {colorUseRule=UseColor_Specified, specifiedColor=<rgba 0 0 0 0>}
      cornerRadii=<0 0 0 0>
     }
     tintColor {colorUseRule=UseColor_Specified, specifiedColor=<rgba 1 1 1 1>}
     uVRegion {bIsValid=false, max=<0 0>, min=<0 0>}
    }}
   slot {class=ButtonSlot}
  }
  Button_Delete {
   widgetClassPath=&EditorUtilityButton
   parent=HorizontalBox_MoveDelete
   widget=Button_Delete
   properties {toolTipText=Delete Key}
   slot {class=HorizontalBoxSlot}
  }
  Image {
   widgetClassPath=&Image
   parent=Button_Delete
   widget=Image
   properties {brush {
     bIsDynamicallyLoaded=false
     drawAs=Image
     imageType=NoImage
     mirroring=NoMirror
     resourceName=None
     resourceObject=/Landmass/Landscape/OldPrototype_BP/Mats/Slate/M_Button_DeleteX
     tiling=NoTile
     imageSize=<32 32>
     margin=<ltrb 0 0 0 0>
     outlineSettings {
      bUseBrushTransparency=false
      roundingType=HalfHeightRadius
      width=0
      color {colorUseRule=UseColor_Specified, specifiedColor=<rgba 0 0 0 0>}
      cornerRadii=<0 0 0 0>
     }
     tintColor {colorUseRule=UseColor_Specified, specifiedColor=<rgba 1 1 1 1>}
     uVRegion {bIsValid=false, max=<0 0>, min=<0 0>}
    }}
   slot {class=ButtonSlot}
  }
  Button_SwapFwd {
   widgetClassPath=&EditorUtilityButton
   parent=HorizontalBox_MoveDelete
   widget=Button_SwapFwd
   properties {toolTipText=Move Key to Right}
   slot {class=HorizontalBoxSlot}
  }
  Image_1 {
   widgetClassPath=&Image
   parent=Button_SwapFwd
   widget=Image_1
   properties {brush {
     bIsDynamicallyLoaded=false
     drawAs=Image
     imageType=NoImage
     mirroring=NoMirror
     resourceName=None
     resourceObject=/Landmass/Landscape/OldPrototype_BP/Mats/Slate/M_Button_Arrow_Fwd
     tiling=NoTile
     imageSize=<32 32>
     margin=<ltrb 0 0 0 0>
     outlineSettings {
      bUseBrushTransparency=false
      roundingType=HalfHeightRadius
      width=0
      color {colorUseRule=UseColor_Specified, specifiedColor=<rgba 0 0 0 0>}
      cornerRadii=<0 0 0 0>
     }
     tintColor {colorUseRule=UseColor_Specified, specifiedColor=<rgba 1 1 1 1>}
     uVRegion {bIsValid=false, max=<0 0>, min=<0 0>}
    }}
   slot {class=ButtonSlot}
  }
  VerticalBox_147 {
   widgetClassPath=&VerticalBox
   parent=VerticalBox_440
   widget=VerticalBox_147
   slot {class=VerticalBoxSlot, properties {verticalAlignment=VAlign_Top, size {sizeRule=Fill, value=1}}}
  }
  EditorUtilityScrollBox_453 {
   widgetClassPath=&EditorUtilityScrollBox
   parent=VerticalBox_147
   widget=EditorUtilityScrollBox_453
   slot {class=VerticalBoxSlot, properties {verticalAlignment=VAlign_Top, size {sizeRule=Fill, value=1}}}
  }
  SizeBox_295 {widgetClassPath=&SizeBox, parent=EditorUtilityScrollBox_453, widget=SizeBox_295, slot {class=ScrollBoxSlot}}
  VerticalBox_254 {widgetClassPath=&VerticalBox, parent=SizeBox_295, widget=VerticalBox_254, slot {class=SizeBoxSlot}}
  Details_ParticleErosion {
   widgetClassPath=&DetailsView
   parent=VerticalBox_254
   widget=Details_ParticleErosion
   properties {visibility=Collapsed, propertiesToShow [Key Name, Erosion Mode, Particle Erosion Settings]}
   slot {class=VerticalBoxSlot}
  }
  Details_NoiseInjection {
   widgetClassPath=&DetailsView
   parent=VerticalBox_254
   widget=Details_NoiseInjection
   properties {visibility=Collapsed, propertiesToShow [Key Name, Erosion Mode, Noise Injection Settings]}
   slot {class=VerticalBoxSlot}
  }
  Details_CustomMaterial {
   widgetClassPath=&DetailsView
   parent=VerticalBox_254
   widget=Details_CustomMaterial
   properties {visibility=Collapsed, propertiesToShow [Key Name, Erosion Mode, Custom Material Pass]}
   slot {class=VerticalBoxSlot}
  }
  Details_SlopeSmoothing {
   widgetClassPath=&DetailsView
   parent=VerticalBox_254
   widget=Details_SlopeSmoothing
   properties {visibility=Collapsed, propertiesToShow [Key Name, Erosion Mode, Slope Smoothing Settings]}
   slot {class=VerticalBoxSlot}
  }
  Details_WaterLevelling {
   widgetClassPath=&DetailsView
   parent=VerticalBox_254
   widget=Details_WaterLevelling
   properties {visibility=Collapsed, propertiesToShow [Key Name, Erosion Mode, Water Levelling Settings]}
   slot {class=VerticalBoxSlot}
  }
  Button_DataAssetCategory {
   widgetClassPath=&EditorUtilityButton
   parent=VerticalBox_254
   widget=Button_DataAssetCategory
   properties {widgetStyle {
     clickedSlateSound {resourceObject=null}
     disabled {
      bIsDynamicallyLoaded=false
      drawAs=RoundedBox
      imageType=NoImage
      mirroring=NoMirror
      resourceName=None
      resourceObject=null
      tiling=NoTile
      imageSize=<0 0>
      margin=<ltrb 0 0 0 0>
      outlineSettings {
       bUseBrushTransparency=false
       roundingType=FixedRadius
       width=1
       color {colorUseRule=UseColor_ColorTable, specifiedColor=<rgba 0 0 0 0>}
       cornerRadii=<4 4 4 4>
      }
      tintColor {colorUseRule=UseColor_Specified, specifiedColor=<rgba 0.0395 0.0395 0.0395 1>}
      uVRegion {bIsValid=false, max=<0 0>, min=<0 0>}
     }
     disabledForeground {colorUseRule=UseColor_ColorTable, specifiedColor=<rgba 0 0 0 0>}
     hovered {
      bIsDynamicallyLoaded=false
      drawAs=RoundedBox
      imageType=NoImage
      mirroring=NoMirror
      resourceName=None
      resourceObject=null
      tiling=NoTile
      imageSize=<32 32>
      margin=<ltrb 0 0 0 0>
      outlineSettings {
       bUseBrushTransparency=false
       roundingType=FixedRadius
       width=1
       color {colorUseRule=UseColor_ColorTable, specifiedColor=<rgba 0 0 0 0>}
       cornerRadii=<4 4 4 4>
      }
      tintColor {colorUseRule=UseColor_Specified, specifiedColor=<rgba 0.0953 0.0953 0.0953 1>}
      uVRegion {bIsValid=false, max=<0 0>, min=<0 0>}
     }
     hoveredForeground {colorUseRule=UseColor_ColorTable, specifiedColor=<rgba 0 0 0 0>}
     hoveredSlateSound {resourceObject=null}
     normal {
      bIsDynamicallyLoaded=false
      drawAs=RoundedBox
      imageType=NoImage
      mirroring=NoMirror
      resourceName=None
      resourceObject=null
      tiling=NoTile
      imageSize=<32 32>
      margin=<ltrb 0 0 0 0>
      outlineSettings {
       bUseBrushTransparency=false
       roundingType=FixedRadius
       width=1
       color {colorUseRule=UseColor_ColorTable, specifiedColor=<rgba 0 0 0 0>}
       cornerRadii=<4 4 4 4>
      }
      tintColor {colorUseRule=UseColor_Specified, specifiedColor=<rgba 0.0395 0.0395 0.0395 1>}
      uVRegion {bIsValid=false, max=<0 0>, min=<0 0>}
     }
     normalForeground {colorUseRule=UseColor_ColorTable, specifiedColor=<rgba 0 0 0 0>}
     normalPadding=<ltrb 4 1.5 4 1.5>
     pressed {
      bIsDynamicallyLoaded=false
      drawAs=RoundedBox
      imageType=NoImage
      mirroring=NoMirror
      resourceName=None
      resourceObject=null
      tiling=NoTile
      imageSize=<32 32>
      margin=<ltrb 0 0 0 0>
      outlineSettings {
       bUseBrushTransparency=false
       roundingType=FixedRadius
       width=1
       color {colorUseRule=UseColor_ColorTable, specifiedColor=<rgba 0 0 0 0>}
       cornerRadii=<4 4 4 4>
      }
      tintColor {colorUseRule=UseColor_Specified, specifiedColor=<rgba 0.0284 0.0284 0.0284 1>}
      uVRegion {bIsValid=false, max=<0 0>, min=<0 0>}
     }
     pressedForeground {colorUseRule=UseColor_ColorTable, specifiedColor=<rgba 0 0 0 0>}
     pressedPadding=<ltrb 4 1.5 4 1.5>
     pressedSlateSound {resourceObject=null}
    }}
   slot {class=VerticalBoxSlot}
  }
  HorizontalBox_99 {
   widgetClassPath=&HorizontalBox
   parent=Button_DataAssetCategory
   widget=HorizontalBox_99
   slot {class=ButtonSlot, properties {horizontalAlignment=HAlign_Fill, verticalAlignment=VAlign_Fill, padding=<ltrb 0 0 0 0>}}
  }
  Image_SettingsPresetArrow {
   widgetClassPath=&Image
   bIsVariable=true
   parent=HorizontalBox_99
   widget=Image_SettingsPresetArrow
   properties {brush {
     bIsDynamicallyLoaded=false
     drawAs=Image
     imageType=NoImage
     mirroring=NoMirror
     resourceName=None
     resourceObject=&CategoryArrow_Rotated
     tiling=NoTile
     imageSize=<12 12>
     margin=<ltrb 0 0 0 0>
     outlineSettings {
      bUseBrushTransparency=false
      roundingType=HalfHeightRadius
      width=0
      color {colorUseRule=UseColor_Specified, specifiedColor=<rgba 0 0 0 0>}
      cornerRadii=<0 0 0 0>
     }
     tintColor {colorUseRule=UseColor_Specified, specifiedColor=<rgba 0.375 0.375 0.375 1>}
     uVRegion {bIsValid=false, max=<0 0>, min=<0 0>}
    }}
   slot {class=HorizontalBoxSlot, properties {horizontalAlignment=HAlign_Center, verticalAlignment=VAlign_Center}}
  }
  TextBlock_13 {
   widgetClassPath=&TextBlock
   parent=HorizontalBox_99
   widget=TextBlock_13
   properties {
    text=Preset DataAsset
    colorAndOpacity {colorUseRule=UseColor_Specified, specifiedColor=<rgba 0.6667 0.6667 0.6667 1>}
    font {
     bForceMonospaced=false
     bMaterialIsStencil=false
     fontMaterial=null
     fontObject=&Roboto
     letterSpacing=0
     monospacedWidth=1
     size=8
     skewAmount=0
     typefaceFontName=Bold
     outlineSettings {
      bApplyOutlineToDropShadows=false
      bMiteredCorners=false
      bSeparateFillAlpha=false
      outlineMaterial=null
      outlineSize=0
      outlineColor=<rgba 0 0 0 1>
     }
    }
   }
   slot {class=HorizontalBoxSlot, properties {horizontalAlignment=HAlign_Left, verticalAlignment=VAlign_Center, padding=<ltrb 4 4 0 4>}}
  }
  VerticalBox_SettingsPreset {
   widgetClassPath=&VerticalBox
   bIsVariable=true
   parent=VerticalBox_254
   widget=VerticalBox_SettingsPreset
   properties {visibility=Visible}
   slot {class=VerticalBoxSlot}
  }
  HorizontalBox_69 {
   widgetClassPath=&HorizontalBox
   parent=VerticalBox_SettingsPreset
   widget=HorizontalBox_69
   slot {class=VerticalBoxSlot, properties {horizontalAlignment=HAlign_Center}}
  }
  Overlay_237 {widgetClassPath=&Overlay, parent=HorizontalBox_69, widget=Overlay_237, slot {class=HorizontalBoxSlot}}
  HorizontalBox_PresetButtons {
   widgetClassPath=&HorizontalBox
   parent=Overlay_237
   widget=HorizontalBox_PresetButtons
   slot {class=OverlaySlot, properties {horizontalAlignment=HAlign_Fill, verticalAlignment=VAlign_Fill}}
  }
  Button_LoadSettings {
   widgetClassPath=&EditorUtilityButton
   parent=HorizontalBox_PresetButtons
   widget=Button_LoadSettings
   properties {toolTipText=Loads and applies all settings from the specified Settings Preset Data Asset to this instance.}
   slot {class=HorizontalBoxSlot, properties {padding=<ltrb 10 10 10 10>}}
  }
  TextBlock_10 {
   widgetClassPath=&TextBlock
   parent=Button_LoadSettings
   widget=TextBlock_10
   properties {
    text=Load Settings
    font {
     bForceMonospaced=false
     bMaterialIsStencil=false
     fontMaterial=null
     fontObject=&Roboto
     letterSpacing=0
     monospacedWidth=1
     size=10
     skewAmount=0
     typefaceFontName=Bold
     outlineSettings {
      bApplyOutlineToDropShadows=false
      bMiteredCorners=false
      bSeparateFillAlpha=false
      outlineMaterial=null
      outlineSize=0
      outlineColor=<rgba 0 0 0 1>
     }
    }
   }
   slot {class=ButtonSlot, properties {padding=<ltrb 0 4 0 4>}}
  }
  Button_SaveToDataAsset {
   widgetClassPath=&EditorUtilityButton
   parent=HorizontalBox_PresetButtons
   widget=Button_SaveToDataAsset
   properties {toolTipText=Saves the current settings from this instance to the specified Settings Preset Data Asset.}
   slot {class=HorizontalBoxSlot, properties {padding=<ltrb 0 10 0 10>}}
  }
  TextBlock_11 {
   widgetClassPath=&TextBlock
   parent=Button_SaveToDataAsset
   widget=TextBlock_11
   properties {
    text=Apply to DataAsset
    font {
     bForceMonospaced=false
     bMaterialIsStencil=false
     fontMaterial=null
     fontObject=&Roboto
     letterSpacing=0
     monospacedWidth=1
     size=10
     skewAmount=0
     typefaceFontName=Bold
     outlineSettings {
      bApplyOutlineToDropShadows=false
      bMiteredCorners=false
      bSeparateFillAlpha=false
      outlineMaterial=null
      outlineSize=0
      outlineColor=<rgba 0 0 0 1>
     }
    }
   }
   slot {class=ButtonSlot, properties {padding=<ltrb 0 4 0 4>}}
  }
  Button_NewDataAsset {
   widgetClassPath=&EditorUtilityButton
   parent=HorizontalBox_PresetButtons
   widget=Button_NewDataAsset
   properties {toolTipText=Creates a new Settings Data Asset using the settings specified here. Default path will be copied from the specified Source Mesh.}
   slot {class=HorizontalBoxSlot, properties {padding=<ltrb 10 10 10 10>}}
  }
  TextBlock_12 {
   widgetClassPath=&TextBlock
   parent=Button_NewDataAsset
   widget=TextBlock_12
   properties {
    text=New DataAsset
    font {
     bForceMonospaced=false
     bMaterialIsStencil=false
     fontMaterial=null
     fontObject=&Roboto
     letterSpacing=0
     monospacedWidth=1
     size=10
     skewAmount=0
     typefaceFontName=Bold
     outlineSettings {
      bApplyOutlineToDropShadows=false
      bMiteredCorners=false
      bSeparateFillAlpha=false
      outlineMaterial=null
      outlineSize=0
      outlineColor=<rgba 0 0 0 1>
     }
    }
   }
   slot {class=ButtonSlot, properties {padding=<ltrb 0 4 0 4>}}
  }
  NewDataAsset_PathPrompt {
   widgetClassPath=&Border
   bIsVariable=true
   parent=Overlay_237
   widget=NewDataAsset_PathPrompt
   properties {visibility=Collapsed, brushColor=<rgba 0.018 0.018 0.018 1>, padding=<ltrb 0 0 0 0>}
   slot {class=OverlaySlot, properties {horizontalAlignment=HAlign_Fill, verticalAlignment=VAlign_Fill}}
  }
  HorizontalBox_482 {
   widgetClassPath=&HorizontalBox
   parent=NewDataAsset_PathPrompt
   widget=HorizontalBox_482
   slot {class=BorderSlot, properties {padding=<ltrb 0 0 0 0>}}
  }
  NewDataAssetPath {
   widgetClassPath=&EditorUtilityEditableTextBox
   bIsVariable=true
   parent=HorizontalBox_482
   widget=NewDataAssetPath
   properties {text=/Game/Landmass/Presets/ErosionPreset_New}
   slot {class=HorizontalBoxSlot, properties {verticalAlignment=VAlign_Center}}
  }
  EditorUtilityButton_CreateAsset {
   widgetClassPath=&EditorUtilityButton
   parent=HorizontalBox_482
   widget=EditorUtilityButton_CreateAsset
   slot {class=HorizontalBoxSlot, properties {horizontalAlignment=HAlign_Right, padding=<ltrb 4 2 4 2>, size {sizeRule=Fill, value=1}}}
  }
  TextBlock {
   widgetClassPath=&TextBlock
   parent=EditorUtilityButton_CreateAsset
   widget=TextBlock
   properties {
    text=Create Asset
    font {
     bForceMonospaced=false
     bMaterialIsStencil=false
     fontMaterial=null
     fontObject=&Roboto
     letterSpacing=0
     monospacedWidth=1
     size=10
     skewAmount=0
     typefaceFontName=Bold
     outlineSettings {
      bApplyOutlineToDropShadows=false
      bMiteredCorners=false
      bSeparateFillAlpha=false
      outlineMaterial=null
      outlineSize=0
      outlineColor=<rgba 0 0 0 1>
     }
    }
   }
   slot {class=ButtonSlot, properties {padding=<ltrb 0 4 0 4>}}
  }
  SinglePropertyView_80 {
   widgetClassPath=&SinglePropertyView
   parent=VerticalBox_SettingsPreset
   widget=SinglePropertyView_80
   properties {
    propertyName=SettingsPreset
    toolTipText="This is a Data Asset that contains all the properties for Impostor Generation. Allows saving and reusing presets.\r\n\r\nA default preset can be specified per-project by adding the following to DefaultEditor.ini:\r\n\r\n[/ImpostorBaker/BP/EUW_Generate_Impostor.EUW_Generate_Impostor_C]\r\nDefaultSettingsAssetPath=/Game/YourPresetAssetName\r\n"
   }
   slot {class=VerticalBoxSlot, properties {padding=<ltrb 15 0 15 0>}}
  }
 ]
}
