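# Editor utility Blueprint that opens Landmass editor widget tabs and links them to the brushes they edit.
# Spawn Erosion Widget registers the EUW_Landscape_Erosion tab and assigns the given erosion brush as its owner.
# Spawn Landmass Editor registers the EUW_Landmass_Editor tab, cross-links it with the brush manager, and sets the selected brush after a validity check.
# Both events use EditorUtilitySubsystem and cast the spawned widget before setting references.
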EditorUtilityBlueprint /Landmass/Landscape/ErosionBrush/BP/Widgets/EUB_SpawnErosionTab
graphs [
 EventGraph {nodes [
   n1 = Spawn Erosion Widget!() -> (OutputDelegate, then, OwningBrush:Landmass Brush Erosion Object Reference) {is=title K2Node_CustomEvent}
   > n2 = SpawnAndRegisterTab(self=Get EditorUtilitySubsystem(), InBlueprint=&/EUW_Landscape_Erosion)
   > n3 = Cast To EUW_Landscape_Erosion!(execute, Object=@n2) -> (then, CastFailed, AsEUW Landscape Erosion:EUW Landscape Erosion Object Reference) {is=title K2Node_DynamicCast}
   > Set Owning_ErosionBrush!(execute, Owning_ErosionBrush:Landmass Brush Erosion Object Reference=@n1.OwningBrush, self:EUW Landscape Erosion Object Reference=@n3) -> (then, Output_Get:Landmass Brush Erosion Object Reference)
   n4 = Spawn Landmass Editor!() -> (OutputDelegate, then, Manager:Landmass Brush Manager Object Reference, LandmassBrush:Landmass Actor Object Reference) {is=title K2Node_CustomEvent}
   > n5 = SpawnAndRegisterTab(self=Get EditorUtilitySubsystem(), InBlueprint=/Landmass/Landscape/BlueprintBrushes/Widgets/EUW_Landmass_Editor)
   > n6 = Cast To EUW_Landmass_Editor!(execute, Object=@n5) -> (then, CastFailed, AsEUW Landmass Editor:EUW Landmass Editor Object Reference) {is=title K2Node_DynamicCast}
   > Set BrushManager!(execute, BrushManager:Landmass Brush Manager Object Reference=@n4.Manager, self:EUW Landmass Editor Object Reference=@n6) -> (then, Output_Get:Landmass Brush Manager Object Reference)
   > Set EditorWidget!(execute, EditorWidget:EUW Landmass Editor Object Reference=@n6, self:Landmass Brush Manager Object Reference=@n4.Manager) -> (then, Output_Get:EUW Landmass Editor Object Reference)
   > Is Valid(InputObject=@n4.LandmassBrush)
   > Set SelectedBrush!(execute, SelectedBrush:Landmass Actor Object Reference=@n4.LandmassBrush, self:EUW Landmass Editor Object Reference=@n6) -> (then, Output_Get:Landmass Actor Object Reference)
   Event BeginPlay()
   Event ActorBeginOverlap()
   Event Tick!() -> (OutputDelegate, then, DeltaSeconds:float)
  ]}
 UserConstructionScript {nodes [Construction Script()]}
]
variables []
